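# Translucent unlit material for triangle FX in the YuzuCrank emote, enabled for Niagara mesh particles.
# A Fresnel term with an adjustable exponent multiplies particle color RGB for tint and particle alpha for opacity, weighting the effect toward glancing edges.
# The alpha product passes through a parameterized depth fade to soften intersections with scene geometry.
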Material /BRCosmetics/Animation/Game/MainPlayer/Emotes/YuzuCrank/FX/M_Emote_YuzuCrank_Triangle
graph {
 properties {
  bAutomaticallySetUsageInEditor=false
  bUsedWithNiagaraMeshParticles=true
  blendMode=BLEND_Translucent
  shadingModel=MSM_Unlit
  usedShadingModels=Unlit
 }
 nodes [
  n1 = Fresnel(ExponentIn=ScalarParameter() {defaultValue=1, parameterName=Fresnel Strength})
  n2 = ParticleColor()
  DepthFade(Opacity=Multiply(A=@n1, B=@n2.A), FadeDistance=ScalarParameter() {defaultValue=300, parameterName=Depth Fade})
  Multiply(A=@n1, B=@n2.RGB)
 ]
}
