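# Niagara system for the YuzuCrank emote's ground laser effect, timed mostly by delayed emitter loops.
# An iridescent mesh cylinder fades over loop age while purple mesh rings rise through it.
# A decal glow, drifting dust sprites, two light renderers and player glow sprites complete the look.
# Time-of-day scalars tune brightness, with front-end, platform and distance modifiers on some layers; Box_Start and Box_Mid have no renderers.
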NiagaraSystem /BRCosmetics/Animation/Game/MainPlayer/Emotes/YuzuCrank/FX/NS_Emote_YuzuCrank_GroundLaser
dependencies {
 usedDataInterfaces [
  &NiagaraDataInterfaceCurve
  &NiagaraDataInterfaceColorCurve
  &NiagaraDataInterfaceVectorCurve
  &NiagaraDataInterfacePlatformSet
 ]
 usedDynamicInputs [&Multiply_VectorByFloat]
 usedModules [&SystemState, &BlendSpaceHelper]
}
emitters {
 Box_Mid {rendererClasses [], renderers []}
 Box_Start {rendererClasses [], renderers []}
 Decal_Glow {
  emitterUpdateScript [
   &EmitterState(
    Inactive Response=::NewEnumerator1
    Life Cycle Mode=::NewEnumerator1
    Loop Behavior=::NewEnumerator1
    Loop Delay=1
    Loop Duration=3
    UseLoopDelay=true
   )
   &SpawnBurst_Instantaneous(Use Spawn Probability=false, UseLoopCountLimit=false)
  ]
  particleSpawnScript [
   &InitializeParticle(
    Color=&Multiply_LinearColorByFloat(
     Linear Color=<rgba 0.7807 0.9114 1 1>
     Float=&FNTimeOfDayScalar(Night Scale=1, Morning Scale=0.7, Day Scale=0.7, Evening Scale=0.7)
    )
    Lifetime=3.5
    UsePositionOffset=false
    Write Material Random=true
   )
   /Niagara/Modules/Decal/Decal_Attributes(Decal Size:Vector3f=<150 500 500>, Size=1)
  ]
  particleUpdateScript [
   &ParticleState()
   &ScaleColor(
    Scale Alpha=&FloatFromCurve(FloatCurve=DataInterface() {propertyValues {Curve {keys [<at 0 1.5>,<at 0.05 1>,<at 0.8 0>]}}})
    ScaleA=true
    ScaleRGB=true
   )
   &BlendSpace()
  ]
  rendererClasses [&NiagaraDecalRendererProperties]
  renderers [{rendererClass=&NiagaraDecalRendererProperties, rendererIndex=0}]
 }
 Dust {
  emitterUpdateScript [
   &EmitterState(Life Cycle Mode=::NewEnumerator1, Loop Behavior=::NewEnumerator1, Loop Delay=0.8, Loop Duration=1.5, UseLoopDelay=true)
   &SpawnBurst_Instantaneous(Spawn Count=30, Use Spawn Probability=false, UseLoopCountLimit=false)
   &SpawnRate(SpawnRate=60, Use Spawn Probability=false)
  ]
  particleSpawnScript [
   &InitializeParticle(
    Color=&FNTimeOfDayColor(
     Night Color=<rgba 0.8782 0.9249 0.9635 0.25>
     Morning Color=<rgba 0.2109 0.3263 0.4219 0.08>
     Day Color=<rgba 0.2109 0.3263 0.4219 0.08>
     Evening Color=<rgba 0.2109 0.3263 0.4219 0.08>
    )
    Lifetime Min=1.5
    Lifetime Mode=::NewEnumerator1
    Sprite Rotation Angle=&RandomRangeFloat(Minimum=-0.05, Maximum=0.3)
    Sprite Rotation Mode=::NewEnumerator1
    Sprite Size Mode=::NewEnumerator1
    Uniform Sprite Size Max=50
    Uniform Sprite Size Min=60
    UsePositionOffset=true
    Write Material Random=true
   )
   &ShapeLocation(Cylinder Height=15, MaskSpawnGroup=false, Override Seed=false, Shape Primitive=::NewEnumerator1, Surface Only=false, UseOwnerScale=true)
   &AddVelocity(Velocity=&RandomRangeVector(Minimum=<-2 -2 20>, Maximum=<2 2 50>, Evaluation Type=::NewEnumerator1))
   &InheritVelocity(Module.Limit Speed=true, Velocity Limit=5, Velocity Scale=<0.25 0.25 0.25>) {enabled=false}
   &PointForce(Force Strength=25, Module.Use Falloff=false, Module.UseLinearForce=false)
   &ApplyInitialForces()
   /SetVariables(Particles.LayerColor:LinearColor=<rgba 1 0.6847 0.3958 0.5>, Particles.LayerScale:Vector2f=<40 40>) {isSetParameters=true}
  ]
  particleUpdateScript [
   &ParticleState()
   &SubUVAnimation(
    Sub UV Play Rate:NiagaraFloat=&RandomRangeFloat(Minimum=0.03, Maximum=0.05)
    SubUV Animation Mode:ENiagara_SubUVLookupMode=::NewEnumerator3
    SubUV Start Time Offset:NiagaraFloat=&RandomRangeFloat(Minimum=0, Maximum=63)
   )
   &DynamicMaterialParameters(
    Index 0 Param 1=&FloatFromCurve(FloatCurve=DataInterface() {propertyValues {Curve {keys [<at 0 -0.25>,<at 1 -1>]}}})
    Index 0 Param 4=&OneMinusFloat(Float=$Particles.NormalizedAge)
    Module.Param0WriteEnabled=true
    Module.Param1WriteEnabled=false
    Module.Param2WriteEnabled=false
    Module.Param3WriteEnabled=true
   )
   &ScaleSpriteSize(Scale Factor=&Vector2DFromFloat(Value=&FloatFromCurve(
      FloatCurve=DataInterface() {propertyValues {Curve {keys [
          {arriveTangent=1.8751, interpMode=RCIM_Cubic, leaveTangent=1.8751, tangentMode=RCTM_User, time=0, value=0.5}
          {arriveTangent=0.72, interpMode=RCIM_Cubic, leaveTangent=0.72, tangentMode=RCTM_User, time=1, value=1.25}
         ]}}}
      Scale Curve=1.5
     )))
   &Drag(Drag=0.5, Use Linear Drag=true, Use Rotational Drag=false)
   &InheritSourceMovement(ApplyForce=false, ApplyPosition=true, ApplyPositionFalloff=false)
   &SolveForcesAndVelocity(Clamp Velocity=false, Limit Acceleration=false)
   &ScaleColor(
    Scale Alpha=&FloatFromCurve(FloatCurve=DataInterface() {propertyValues {Curve {keys [<at 0 0>,<at 0.3 1>,<at 0.85 1>,<at 1 0>]}}})
    ScaleA=true
    ScaleRGB=true
   )
   &ScaleColor(ScaleA=true, ScaleRGB=true) {module=ScaleColor001}
   &BlendSpace() {enabled=false}
  ]
  rendererClasses [&NiagaraSpriteRendererProperties]
  renderers [
   {rendererClass=&NiagaraSpriteRendererProperties}
   {rendererClass=&NiagaraSpriteRendererProperties, rendererIndex=1}
  ]
 }
 Iridescent_Cylinder {
  emitterUpdateScript [
   &EmitterState(
    Inactive Response=::NewEnumerator1
    Life Cycle Mode=::NewEnumerator1
    Loop Behavior=::NewEnumerator1
    Loop Delay=1
    Loop Duration=3.5
    UseLoopDelay=true
   )
   &SpawnBurst_Instantaneous(Use Spawn Probability=false, UseLoopCountLimit=false)
   /SetVariables(Emitter.EmitterEdgeColor:LinearColor=&ColorFromCurve(
     ColorCurve=DataInterface() {propertyValues {
       AlphaCurve {keys [
         <at 0 0>
         <at 0.0819 0.01>
         {arriveTangent=0.0266, interpMode=RCIM_Cubic, leaveTangent=0.0266, time=0.2315, value=0.02}
         <at 0.7513 0.02>
         <at 0.8257 0 RCIM_Cubic>
        ]}
       BlueCurve {keys [<at 0 0.6867 RCIM_Cubic>,<at 1 0.6867 RCIM_Cubic>]}
       GreenCurve {keys [<at 0 0.5795 RCIM_Cubic>,<at 1 0.5651 RCIM_Cubic>]}
       RedCurve {keys [<at 0 0.467 RCIM_Cubic>,<at 1 0.3855 RCIM_Cubic>]}
      }}
     CurveIndex=$Emitter.NormalizedLoopAge
    )) {isSetParameters=true}
  ]
  particleSpawnScript [&InitializeParticle(
    Lifetime=8
    Mass Mode=::NewEnumerator3
    Mesh Scale=<7.5 7.5 14>
    Mesh Scale Mode=::NewEnumerator4
    Position Offset=<0 0 0.3>
    Sprite Rotation Mode=::NewEnumerator2
    UsePositionOffset=true
    Write Material Random=true
   )]
  particleUpdateScript [
   &ParticleState()
   &ScaleMeshSize(Scale Factor=&VectorFromCurve(VectorCurve=DataInterface() {propertyValues {XCurve {keys [<at 0 1>,<at 0.9 1>,<at 1 0>]}, YCurve {keys [<at 0 1>,<at 0.9 1>,<at 1 0>]}, ZCurve {keys [<at 0 1>,<at 1 1>]}}}))
   &Drag(Drag=0.25, Use Linear Drag=true, Use Rotational Drag=false)
   /SetVariables(Particles.IridescentColor:LinearColor=&Multiply_LinearColorByFloat(
     Linear Color=&ColorFromCurve(
      ColorCurve=DataInterface() {propertyValues {
        AlphaCurve {keys [<at 0 0>,<at 0.0819 0.65>,<at 0.2315 0.65 RCIM_Cubic>,<at 0.7513 0.65>,<at 0.8257 0 RCIM_Cubic>]}
        BlueCurve {keys [<at 0 0.6867 RCIM_Cubic>,<at 1 0.6867 RCIM_Cubic>]}
        GreenCurve {keys [<at 0 0.4508 RCIM_Cubic>,<at 1 0.4813 RCIM_Cubic>]}
        RedCurve {keys [<at 0 0.2601 RCIM_Cubic>,<at 1 0.2625 RCIM_Cubic>]}
       }}
      CurveIndex=$Emitter.NormalizedLoopAge
     )
     Float=&FNTimeOfDayScalar(Night Scale=0.5, Morning Scale=1, Evening Scale=1)
    )) {isSetParameters=true}
   /SetVariables(Particles.Vector:Vector3f=&Multiply_Vector(A=$Particles.Scale, B=<0.25 0.25 1>)) {isSetParameters=true}
   &InheritSourceMovement(ApplyForce=false, ApplyPosition=true, ApplyPositionFalloff=false)
   &SolveForcesAndVelocity(Clamp Velocity=false, Limit Acceleration=false)
   &BlendSpace() {enabled=false}
  ]
  rendererClasses [&NiagaraMeshRendererProperties]
  renderers [
   {rendererClass=&NiagaraMeshRendererProperties, rendererIndex=0}
   {rendererClass=&NiagaraMeshRendererProperties, rendererIndex=1}
  ]
 }
 Light {
  emitterUpdateScript [
   &EmitterState(Life Cycle Mode=::NewEnumerator1, Loop Behavior=::NewEnumerator1, Loop Duration=6, UseLoopDelay=false)
   &SpawnBurst_Instantaneous(Use Spawn Probability=false, UseLoopCountLimit=false)
  ]
  particleSpawnScript [&InitializeParticle(Color Mode=::NewEnumerator0, Lifetime=5, UsePositionOffset=true, Write Material Random=true)]
  particleUpdateScript [
   &ParticleState()
   &SystemLocation(MaskSpawnGroup=false, Offset=<0 0 40>, Offset Coordinate Space=::Simulation)
   &FN_Light_Attributes(
    Color=&Multiply_LinearColorByFloat(
     Linear Color=<rgba 5.4583 7.917 8 1>
     Float=&FNTimeOfDayScalar(Night Scale=0.7, Morning Scale=0.6, Day Scale=0.3, Evening Scale=0.6)
    )
    Create Color Attribute=true
    Create Exponent Attribute=true
    Create Position Attribute=true
    Create Radius Attribute=true
    Exponent=5
    Position=<5 0 0>
    Radius=500
    Scale Color=&FloatFromCurve(FloatCurve=DataInterface() {propertyValues {Curve {keys [<at 0 0>,<at 0.21 0.1>,<at 0.26 1>,<at 0.75 1>,<at 0.85 0>]}}})
   )
   &FN_Light_Attributes(
    Color=&Multiply_LinearColorByFloat(
     Linear Color=<rgba 5.4583 7.917 8 1>
     Float=&FNTimeOfDayScalar(Night Scale=0.7, Morning Scale=0.6, Day Scale=0.5, Evening Scale=0.6)
    )
    Create Color Attribute=true
    Create Exponent Attribute=true
    Create Position Attribute=true
    Create Radius Attribute=true
    Exponent=5
    Position=<5 0 0>
    Radius=500
    Scale Color=&FloatFromCurve(FloatCurve=DataInterface() {propertyValues {Curve {keys [<at 0 0>,<at 0.21 0.1>,<at 0.26 1>,<at 0.75 1>,<at 0.85 0>]}}})
   ) {module=FN_Light_Attributes001}
   /SetVariables(Particles.UpLIghtPos:NiagaraPosition=&AddVectorToPosition(Position=<0 0 0>, Vector=<0 0 250>)) {isSetParameters=true}
  ]
  rendererClasses [&NiagaraLightRendererProperties]
  renderers [
   {rendererClass=&NiagaraLightRendererProperties, rendererIndex=0}
   {rendererClass=&NiagaraLightRendererProperties, rendererIndex=1}
  ]
 }
 Player_Light {
  emitterUpdateScript [
   &EmitterState(
    Inactive Response=::NewEnumerator1
    Life Cycle Mode=::NewEnumerator1
    Loop Behavior=::NewEnumerator1
    Loop Delay=1
    Loop Duration=4.5
    UseLoopDelay=true
   )
   &SpawnBurst_Instantaneous(Use Spawn Probability=false, UseLoopCountLimit=false)
  ]
  particleSpawnScript [
   &InitializeParticle(
    Color=<rgba 0.15 0.15 0.15 1.5>
    Lifetime=7
    Mesh Scale=<0.05 0.08 0.05>
    Mesh Scale Mode=::NewEnumerator4
    Position Offset=<0 0 1>
    Sprite Size=<180 300>
    Sprite Size Mode=::NewEnumerator4
    UsePositionOffset=true
    Write Material Random=true
   )
   &InitialMeshOrientation(Mesh Orientation Mode=::NewEnumerator2, Orientation Vector=<0 0 1>, Rotation=<0 0.25 0>, Use Rotation Vector=true)
   &CameraOffset(Camera Offset Amount=5)
   /SetVariables(Particles.SparkCameraOFfset=-10) {isSetParameters=true}
  ]
  particleUpdateScript [
   &ParticleState()
   /SetVariables(Particles.Size003:Vector2f=&Multiply_Vector2D(A=<150 150>, B=<1.2 2.2>)) {isSetParameters=true}
   /SetVariables(Particles.Size004:Vector2f=&Multiply_Vector2D(A=<150 150>, B=<3 6>)) {isSetParameters=true}
   /SetVariables(Particles.Color003:LinearColor=&MakeLinearColorFromVectorAndFloat(
     "Vector (RGB)"=&MakeVectorFromLinearColorRGB(Color=<rgba 0.3646 0.303 0.3445 1>)
     "Float (Alpha)"=&Multiply_Float(
      A=&FloatFromCurve(
       FloatCurve=DataInterface() {propertyValues {Curve {keys [<at 0 1>,<at 0.1 2>,<at 0.46 0>,<at 1 0>]}}}
       Scale Curve=&FNTimeOfDayScalar(Night Scale=0.2, Morning Scale=0.1, Day Scale=0.22, Evening Scale=0.2)
      )
      B=&Multiply_Float(
       A=&FN_FrontEndFloatModifier(Gameplay=0.5, Front End=0.15)
       B=&FN_PickFloatDependOnPlatform(TargetPlatformActive=1, TargetPlatformNoActive=0.4)
      )
     )
    )) {isSetParameters=true}
   /SetVariables(Particles.Color004:LinearColor=&MakeLinearColorFromVectorAndFloat(
     "Vector (RGB)"=&MakeVectorFromLinearColorRGB(Color=<rgba 0.3646 0.303 0.3445 1>)
     "Float (Alpha)"=&Multiply_Float(
      A=&FloatFromCurve(
       FloatCurve=DataInterface() {propertyValues {Curve {keys [<at 0 1>,<at 0.1 2>,<at 0.46 0>,<at 1 0>]}}}
       Scale Curve=&FNTimeOfDayScalar(Night Scale=0.23, Morning Scale=0.1, Day Scale=0.18, Evening Scale=0.2)
      )
      B=&Multiply_Float(
       A=&FN_FrontEndFloatModifier(Gameplay=0.5, Front End=0.15)
       B=&FN_PickFloatDependOnPlatform(TargetPlatformActive=1, TargetPlatformNoActive=0.4)
      )
     )
    )) {isSetParameters=true}
   &SolveForcesAndVelocity(Clamp Velocity=false, Limit Acceleration=false)
  ]
  rendererClasses [&NiagaraSpriteRendererProperties]
  renderers [
   {rendererClass=&NiagaraSpriteRendererProperties}
   {rendererClass=&NiagaraSpriteRendererProperties, rendererIndex=1}
  ]
 }
 Purple_Rings {
  emitterUpdateScript [
   &EmitterState(Life Cycle Mode=::NewEnumerator1, Loop Behavior=::NewEnumerator1, Loop Delay=1, Loop Duration=2, UseLoopDelay=true)
   &SpawnRate(SpawnRate=3, Use Spawn Probability=false)
   &SpawnBurst_Instantaneous(Spawn Group=1, Use Spawn Probability=true, UseLoopCountLimit=false) {module=SpawnBurst_Instantaneous002}
  ]
  particleSpawnScript [
   &FN_SetSpawnGroup(Number of Spawn Groups=2)
   &InitializeParticle(
    Color=<rgba 0.4892 0.2055 0.8854 0.01>
    Lifetime Min=1.5
    Lifetime Mode=::NewEnumerator1
    Mesh Scale=<10 10 8>
    Mesh Scale Mode=::NewEnumerator4
    UsePositionOffset=true
    Write Material Random=true
   )
   &InitialMeshOrientation(
    Mesh Orientation Mode=::NewEnumerator1
    Rotation=&RandomRangeVector(Minimum=<0 0 0>, Maximum=<0 0 1>)
    Use Rotation Vector=true
   )
   &AddVelocity(Velocity=<0 0 250>)
   &ApplyInitialForces()
  ]
  particleUpdateScript [
   &ParticleState()
   &ScaleMeshSize(Scale Factor=&VectorFromFloat(Value=&FloatFromCurve(
      FloatCurve=DataInterface() {propertyValues {Curve {keys [{arriveTangent=0.2998, interpMode=RCIM_Cubic, leaveTangent=0.2998, tangentMode=RCTM_User, time=0.9983, value=1}]}}}
      Scale Curve=0.8
     )))
   &ScaleColor(
    Scale Alpha=&FloatFromCurve(
     FloatCurve=DataInterface() {propertyValues {Curve {keys [
         {arriveTangent=-29.315, interpMode=RCIM_Cubic, leaveTangent=-29.315, tangentMode=RCTM_User, time=0, value=8}
         {arriveTangent=-3.2846, interpMode=RCIM_Cubic, leaveTangent=-3.2846, tangentMode=RCTM_User, time=0.45, value=0.6}
         <at 1 0>
        ]}}}
     Scale Curve=&Multiply_Float(
      A=&FN_LOD_Distance_Remap(Start Distance=500, End Distance=1500, Minimum Output=0.6, Maximum Output=1, Clamp Output=true)
      B=&Multiply_Float(
       A=&FNTimeOfDayScalar(Night Scale=0.7, Morning Scale=1, Evening Scale=1)
       B=&Multiply_Float(A=&FN_FrontEndFloatModifier(Gameplay=1, Front End=0.8), B=1)
      )
     )
    )
    Scale RGB=&MakeVectorFromLinearColorRGB(Color=&ColorFromCurve(ColorCurve=DataInterface() {propertyValues {
        AlphaCurve {keys []}
        BlueCurve {keys [<at 0 10>,<at 0.7108 10>,<at 0.9062 1>]}
        GreenCurve {keys [<at 0 1.1472>, {arriveTangent=-9.5462, time=0.7108, value=1.1623}, <at 0.9062 0.2147>]}
        RedCurve {keys [<at 0 3.9498>,<at 0.7108 4.8545>,<at 0.9062 0.5634>]}
       }}))
    ScaleA=true
    ScaleRGB=true
   )
   /SetVariables(Particles.GlowColor=&MakeLinearColorFromVectorAndFloat(
     "Vector (RGB)"=&MakeVectorFromLinearColorRGB(Color=&ColorFromCurve(ColorCurve=DataInterface() {propertyValues {
         AlphaCurve {keys []}
         BlueCurve {keys [<at 0.4456 1 RCIM_Cubic>,<at 0.8412 1>]}
         GreenCurve {keys [<at 0.4456 0.3116 RCIM_Cubic>,<at 0.8412 0.1412>]}
         RedCurve {keys [<at 0.4456 0.4002 RCIM_Cubic>,<at 0.8412 0.4797>]}
        }}))
     "Float (Alpha)"=&FloatFromCurve(
      FloatCurve=DataInterface() {propertyValues {Curve {keys [
          {arriveTangent=-1.8164, interpMode=RCIM_Cubic, leaveTangent=-1.8164, tangentMode=RCTM_User, time=0, value=0}
          {arriveTangent=-0.0446, interpMode=RCIM_Cubic, leaveTangent=-0.0446, tangentMode=RCTM_User, time=0.1, value=1}
          {arriveTangent=-1.9449, interpMode=RCIM_Cubic, leaveTangent=-1.9449, tangentMode=RCTM_User, time=0.55, value=0.4055}
          {arriveTangent=-0.1166, interpMode=RCIM_Cubic, leaveTangent=-0.1166, tangentMode=RCTM_User, time=1, value=0}
         ]}}}
      Scale Curve=&Multiply_Float(
       A=0.2
       B=&Multiply_Float(A=&FN_FrontEndFloatModifier(Gameplay=1, Front End=0.7), B=1)
      )
     )
    )) {isSetParameters=true}
   &ScaleColor(
    Scale Alpha=&FloatFromCurve(FloatCurve=DataInterface() {propertyValues {Curve {keys [<at 1 1>,<at 5 0>]}}}, CurveIndex=$Emitter.Age)
    ScaleA=true
    ScaleRGB=true
   ) {module=ScaleColor001}
   &InheritSourceMovement(ApplyForce=false, ApplyPosition=true, ApplyPositionFalloff=false)
   &SolveForcesAndVelocity(Clamp Velocity=false, Limit Acceleration=false)
   &BlendSpace() {enabled=false}
  ]
  rendererClasses [&NiagaraMeshRendererProperties]
  renderers [
   {rendererClass=&NiagaraMeshRendererProperties, rendererIndex=0}
   {rendererClass=&NiagaraMeshRendererProperties, rendererIndex=1}
  ]
 }
}
summary {
 systemName=NS_Emote_YuzuCrank_GroundLaser
 emitters [
  {emitterName=Dust, rendererClasses [&NiagaraSpriteRendererProperties]}
  {emitterName=Iridescent_Cylinder, rendererClasses [&NiagaraMeshRendererProperties]}
  {emitterName=Box_Start, rendererClasses []}
  {emitterName=Decal_Glow, rendererClasses [&NiagaraDecalRendererProperties]}
  {emitterName=Player_Light, rendererClasses [&NiagaraSpriteRendererProperties]}
  {emitterName=Purple_Rings, rendererClasses [&NiagaraMeshRendererProperties]}
  {emitterName=Box_Mid, rendererClasses []}
  {emitterName=Light, rendererClasses [&NiagaraLightRendererProperties]}
 ]
 userVariables [
  User.bIsFrontEnd {type=NiagaraBool}
  User.bIsFrontEndPreview {type=NiagaraBool}
  User.DeltaTimeMax {type=NiagaraFloat, defaultValue=0.0167, description=""}
  User.DustSpawnCurve {type=NiagaraFloat, defaultValue=1, description=""}
  User.Fade {type=NiagaraFloat, defaultValue=0, description=""}
  User.RibbonIDOffset {type=NiagaraInt32}
  User.SkeletalMesh {
   type=NiagaraDataInterfaceSkeletalMesh
   defaultValue {
    struct=NiagaraExt_VariableValue_DataInterface
    value {dataInterface=&self:NiagaraDataInterfaceSkeletalMesh_0}
   }
  }
  User.SwingSectionIndex {type=NiagaraInt32}
  User.TrailLifeTimeMult {type=NiagaraFloat, defaultValue=0.4, description=""}
  User.WidthMult {type=NiagaraFloat, defaultValue=2, description=""}
 ]
}
