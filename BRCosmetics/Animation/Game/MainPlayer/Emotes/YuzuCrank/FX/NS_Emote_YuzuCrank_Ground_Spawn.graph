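# Niagara system for the ground spawn effect of the YuzuCrank emote, a short bluish-white burst built from three emitters.
# Spawn_Board feeds two mesh renderers and a sprite renderer, deriving separate circuit, wave and glow ages from particle age to drive color, size and material curves.
# Spawn_Burst adds an expanding, randomly rotated mesh that fades out.
# Spawn_Flash adds a large camera-offset sprite, dimmed by time of day, front end and platform.
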NiagaraSystem /BRCosmetics/Animation/Game/MainPlayer/Emotes/YuzuCrank/FX/NS_Emote_YuzuCrank_Ground_Spawn
dependencies {
 usedDataInterfaces [
  &NiagaraDataInterfaceCurve
  &NiagaraDataInterfacePlatformSet
  &NiagaraDataInterfaceArrayFloat
  &NiagaraDataInterfaceColorCurve
 ]
 usedDynamicInputs []
 usedModules [&SystemState, &BlendSpaceHelper]
}
emitters {
 Spawn_Board {
  emitterSpawnScript [/SetVariables(Emitter.Size=0.5) {isSetParameters=true}]
  emitterUpdateScript [
   &EmitterState(Inactive Response=::NewEnumerator1, Life Cycle Mode=::NewEnumerator1, Loop Behavior=::NewEnumerator1, Loop Duration=3, UseLoopDelay=true)
   &SpawnBurst_Instantaneous(Use Spawn Probability=false, UseLoopCountLimit=false) {module=SpawnBurst_Instantaneous001}
  ]
  particleSpawnScript [
   &InitializeParticle(
    Color=<rgba 4 4 4 1>
    Lifetime=0.8
    Mesh Scale Mode=::NewEnumerator3
    Mesh Uniform Scale=&Multiply_Float(A=1.3, B=$Emitter.Size)
    Position Offset=<0 0 -2>
    Sprite Rotation Angle=&SelectFloatFromArray(
     Array Sampling Mode=::NewEnumerator0
     Direct Array Index=?
     Float Selection Array=DataInterface() {propertyValues {FloatData [0,0.25,0.5,0.75]}}
     Random Seed=0
    )
    Sprite Rotation Mode=::NewEnumerator1
    Sprite Size Mode=::NewEnumerator3
    Uniform Sprite Size=&Multiply_Float(A=200, B=$Emitter.Size)
    UsePositionOffset=true
    Write Material Random=true
   )
   /SetVariables(Particles.VisibilityTag=$Emitter.SpawnGroup) {isSetParameters=true}
   /SetVariables(Particles.WaveColor:LinearColor=<rgba 1.6729 1.9933 2 1>, Particles.WaveLifetime=0.6) {isSetParameters=true}
   /SetVariables(
    Particles.Glow Lifetime=0.2
    Particles.Glow Size=&Multiply_Vector2D_ByFloat(Float=$Emitter.Size, Vector2D=&Multiply_Vector2D_ByFloat(Float=80, Vector2D=<1 1>))
    Particles.GlowFacing:Vector3f=<0 0 1>
   ) {isSetParameters=true}
   &InitialMeshOrientation(
    Mesh Orientation Mode=::NewEnumerator2
    Rotation=&RandomRangeVector(Minimum=<0 0 0>, Maximum=<0 0 1>)
    Use Rotation Vector=false
   )
   /SetVariables(Particles.Circuit Lifetime=0.4) {isSetParameters=true}
  ]
  particleUpdateScript [
   &ParticleState()
   /SetVariables(Particles.Update Circuit Age:NiagaraFloat=&Divide_Float(A=$Particles.Age, B=$Particles.Circuit Lifetime)) {isSetParameters=true}
   /SetVariables(Particles.Update WaveAge:NiagaraFloat=&Divide_Float(A=$Particles.Age, B=$Particles.WaveLifetime)) {isSetParameters=true}
   /SetVariables(Particles.Update Glow Age:NiagaraFloat=&Divide_Float(A=$Particles.Age, B=$Particles.Glow Lifetime)) {isSetParameters=true}
   &ScaleColor(
    Scale Mode=::NewEnumerator1
    Scale RGBA=&Vector4FromCurve(
     CurveIndex=$Particles.Update Circuit Age
     Vector4Curve=DataInterface() {propertyValues {
       AlphaCurve {keys [<at 0.3174 1>,<at 0.5961 0>]}
       BlueCurve {keys [<at 0 1>,<at 1 0.2>]}
       GreenCurve {keys [<at 0 0.8583>,<at 1 0.1164>]}
       RedCurve {keys [<at 0 0>,<at 1 0>]}
      }}
    )
    ScaleRGBA=true
   )
   &ScaleSpriteSize(Scale Factor=&Vector2DFromFloat(Value=&FloatFromCurve(
      FloatCurve=DataInterface() {propertyValues {Curve {keys [
          {arriveTangent=0.5497, interpMode=RCIM_Cubic, leaveTangent=0.5497, tangentMode=RCTM_User, time=0, value=0.4128}
          {arriveTangent=0.5872, interpMode=RCIM_Cubic, leaveTangent=0.5872, time=0.2754, value=0.8613}
          {interpMode=RCIM_Cubic, tangentMode=RCTM_User, time=1, value=1}
         ]}}}
      CurveIndex=$Particles.Update Circuit Age
     )))
   &DynamicMaterialParameters(
    Index 0 Param 2=&FloatFromCurve(
     FloatCurve=DataInterface() {propertyValues {Curve {keys [
         <at 0 1 RCIM_Cubic>
         {arriveTangent=0.1115, interpMode=RCIM_Cubic, leaveTangent=0.1115, tangentMode=RCTM_User, time=0.1, value=0.5}
         {interpMode=RCIM_Cubic, tangentMode=RCTM_User, time=1, value=1}
        ]}}}
     CurveIndex=$Particles.Update Circuit Age
    )
    Module.Index0Float0Write=true
    Module.Index0Float1Write=true
    Module.Index0Float2Write=true
    Module.Index0Float3Write=true
   )
   /SetVariables(Particles.DY0:Vector4f=&MakeVector4(
     X=0
     Y=&FloatFromCurve(FloatCurve=DataInterface() {propertyValues {Curve {keys [{tangentMode=RCTM_User, time=0, value=0.47}, {tangentMode=RCTM_User, time=1, value=1}]}}})
     Z=0
     W=0
    )) {isSetParameters=true}
   &ScaleMeshSize(Scale Factor=&VectorFromFloat(Value=&FloatFromCurve(
      FloatCurve=DataInterface() {propertyValues {Curve {keys [
          {arriveTangent=4.8748, interpMode=RCIM_Cubic, leaveTangent=4.8748, tangentMode=RCTM_User, time=0, value=1}
          {arriveTangent=0.9672, interpMode=RCIM_Cubic, leaveTangent=0.9672, tangentMode=RCTM_User, time=0.3099, value=1.9838}
          {arriveTangent=0.1926, interpMode=RCIM_Cubic, leaveTangent=0.1926, tangentMode=RCTM_User, time=1, value=2.1883}
         ]}}}
      CurveIndex=$Particles.Update WaveAge
     )))
   /SetVariables(
    Particles.Update Wave Color:LinearColor=&MakeLinearColorFromVectorAndFloat(
     "Vector (RGB)"=&Multiply_VectorByFloat(
      Vector=&MakeVectorFromLinearColorRGB(Color=$Particles.WaveColor)
      Float=&FloatFromCurve(
       FloatCurve=DataInterface() {propertyValues {Curve {
          defaultValue=1
          keys [
           {tangentMode=RCTM_User, time=0, value=1}
           {arriveTangent=-0.0405, interpMode=RCIM_Cubic, leaveTangent=-0.0405, tangentMode=RCTM_User, time=0.7009, value=1}
           {arriveTangent=-8.0089, interpMode=RCIM_Cubic, leaveTangent=-8.0089, tangentMode=RCTM_User, time=1, value=0}
          ]
         }}}
       Scale Curve=15
      )
     )
     "Float (Alpha)"=&FloatFromCurve(
      FloatCurve=DataInterface() {propertyValues {Curve {keys [
          <at 0 0 RCIM_Cubic>
          {arriveTangent=0.4167, interpMode=RCIM_Cubic, leaveTangent=0.4167, time=0.1, value=1}
          {arriveTangent=-1.1111, interpMode=RCIM_Cubic, leaveTangent=-1.1111, time=0.6, value=0.25}
          <at 1 0 RCIM_Cubic>
         ]}}}
      CurveIndex=$Particles.Update WaveAge
      Scale Curve=0.125
     )
    )
    Particles.Update Wave DynamicParameter:Vector4f=&MakeVector4(
     X=&FloatFromCurve(FloatCurve=DataInterface() {propertyValues {Curve {keys [<at 0 1>,<at 1 0>]}}}, CurveIndex=$Particles.Update WaveAge)
     Y=2
     Z=&FloatFromCurve(FloatCurve=DataInterface() {propertyValues {Curve {keys [<at 0.1 1>,<at 1 0>]}}}, CurveIndex=$Particles.Update WaveAge, Scale Curve=6)
     W=&FloatFromCurve(
      FloatCurve=DataInterface() {propertyValues {Curve {keys [
          <at 0 1>
          {arriveTangent=0.3079, interpMode=RCIM_Cubic, leaveTangent=0.3079, tangentMode=RCTM_User, time=0.3, value=0.8}
          <at 1 1 RCIM_Cubic>
         ]}}}
      CurveIndex=$Particles.Update WaveAge
     )
    )
    Particles.Update Wave DynamicParameter001:Vector4f=&MakeVector4(
     X=&FN_PickFloatDependOnPlatform(TargetPlatformActive=180, TargetPlatformNoActive=120)
     Y=&FN_PickFloatDependOnPlatform(TargetPlatformActive=2.5, TargetPlatformNoActive=4)
     Z=0
     W=0
    )
   ) {isSetParameters=true}
   /SetVariables(
    Particles.Update Glow Color:LinearColor=&MakeLinearColorFromVectorAndFloat(
     "Vector (RGB)"=&MakeVectorFromLinearColorRGB(Color=<rgba 0.8923 0.9609 1 1>)
     "Float (Alpha)"=&FloatFromCurve(
      FloatCurve=DataInterface() {propertyValues {Curve {
         defaultValue=1
         keys [
          {interpMode=RCIM_Cubic, tangentMode=RCTM_User, time=0, value=0}
          {interpMode=RCIM_Cubic, tangentMode=RCTM_User, time=0.2399, value=1}
          {tangentMode=RCTM_User, time=1, value=0}
         ]
        }}}
      CurveIndex=$Particles.Update Glow Age
     )
    )
    Particles.Update Glow DynamicParameter:Vector4f=<2 0 0 0>
    Particles.Update Glow Size:Vector2f=&Multiply_Vector2D(
     A=$Particles.Glow Size
     B=&Multiply_Vector2D_ByFloat(
      Float=&FloatFromCurve(
       FloatCurve=DataInterface() {propertyValues {Curve {
          defaultValue=0.367
          keys [
           {interpMode=RCIM_Cubic, tangentMode=RCTM_User, time=0, value=0.367}
           {arriveTangent=-0.1298, interpMode=RCIM_Cubic, leaveTangent=-0.1298, tangentMode=RCTM_User, time=1, value=3.8846}
          ]
         }}}
       CurveIndex=$Particles.Update Glow Age
      )
      Vector2D=<1.25 1.25>
     )
    )
   ) {isSetParameters=true}
   /SetVariables(
    Particles.WaveColor2:LinearColor=&MultiplyLinearColor(
     Linear Color=&ColorFromCurve(ColorCurve=DataInterface() {propertyValues {
        AlphaCurve {keys [<at 0 1>,<at 1 0>]}
        BlueCurve {keys [<at 0 10>,<at 0.15 3>,<at 1 1>]}
        GreenCurve {keys [<at 0 10>,<at 0.15 3>,<at 1 1>]}
        RedCurve {keys [<at 0 10>,<at 0.15 3>,<at 1 1>]}
       }})
     Scale Factor=&MakeVector4_FromLinearColor(LinearColor=<rgba 9.272 9.8065 10 0.05>)
    )
    Particles.WaveDynamicMaterialParameter:Vector4f=&MakeVector4(
     X=&FloatFromCurve(FloatCurve=DataInterface() {propertyValues {Curve {keys [{arriveTangent=-2.1015, interpMode=RCIM_Cubic, leaveTangent=-2.1015, tangentMode=RCTM_User, time=0, value=0.5}, <at 1 -0.15>]}}})
     Y=&FloatFromCurve(FloatCurve=DataInterface() {propertyValues {Curve {keys [{arriveTangent=-0.3464, interpMode=RCIM_Cubic, leaveTangent=-0.3464, tangentMode=RCTM_User, time=0, value=1}, <at 1 0.9>]}}})
     Z=&FloatFromCurve(FloatCurve=DataInterface() {propertyValues {Curve {keys [<at 0 0>,<at 0.15 1>,<at 1 1.2>]}}}, Scale Curve=0.8)
     W=0
    )
    Particles.WaveScale:Vector3f=&VectorFromFloat(Value=&FloatFromCurve(
      FloatCurve=DataInterface() {propertyValues {Curve {keys [{arriveTangent=2.8792, interpMode=RCIM_Cubic, leaveTangent=2.8792, tangentMode=RCTM_User, time=0, value=1}, <at 1 2.2 RCIM_Cubic>]}}}
      Scale Curve=0.625
     ))
   ) {isSetParameters=true}
   &InheritSourceMovement(ApplyForce=false, ApplyPosition=true, ApplyPositionFalloff=false)
   &SolveForcesAndVelocity(Clamp Velocity=false, Limit Acceleration=false)
  ]
  rendererClasses [&NiagaraMeshRendererProperties, &NiagaraSpriteRendererProperties]
  renderers [
   {rendererClass=&NiagaraMeshRendererProperties, rendererIndex=0}
   {rendererClass=&NiagaraSpriteRendererProperties, rendererIndex=1}
   {rendererClass=&NiagaraMeshRendererProperties, rendererIndex=2}
  ]
 }
 Spawn_Burst {
  emitterUpdateScript [
   &EmitterState(Inactive Response=::NewEnumerator1, Life Cycle Mode=::NewEnumerator1, Loop Behavior=::NewEnumerator1, Loop Duration=2.2, UseLoopDelay=true)
   &SpawnRate(SpawnRate=5, Use Spawn Probability=false) {enabled=false}
   &SpawnBurst_Instantaneous(Spawn Count=1, Use Spawn Probability=false)
  ]
  particleSpawnScript [
   &InitializeParticle(
    Color=<rgba 0.6112 0.7216 1 1>
    Lifetime=0.7
    Mesh Scale Mode=::NewEnumerator3
    Position Offset=<0 0 1>
    UsePositionOffset=true
    Write Material Random=true
   ) {module=InitializeParticle001}
   &InitialMeshOrientation(
    Orientation Axis=<0 0 1>
    Orientation Vector=<0 0 1>
    Rotation=&MakeVector(X=0, Y=0, Z=&RandomRangeFloat(Minimum=0, Maximum=1))
    Use Orientation Vector=true
    Use Rotation Vector=true
   )
  ]
  particleUpdateScript [
   &DynamicMaterialParameters(
    Index 0 Param 1=&FloatFromCurve(FloatCurve=DataInterface() {propertyValues {Curve {keys [{arriveTangent=-2.1015, interpMode=RCIM_Cubic, leaveTangent=-2.1015, tangentMode=RCTM_User, time=0, value=0.5}, <at 1 -0.15>]}}})
    Index 0 Param 2=&FloatFromCurve(FloatCurve=DataInterface() {propertyValues {Curve {keys [{arriveTangent=-0.3464, interpMode=RCIM_Cubic, leaveTangent=-0.3464, tangentMode=RCTM_User, time=0, value=1}, <at 1 0.9>]}}})
    Index 0 Param 3=&FloatFromCurve(FloatCurve=DataInterface() {propertyValues {Curve {keys [<at 0 0>,<at 0.15 1>,<at 1 1.2>]}}})
    Module.Param0WriteEnabled=true
    Module.Param1WriteEnabled=true
    Module.Param2WriteEnabled=true
    Module.Param3WriteEnabled=false
   )
   &ScaleColor(
    Scale Alpha=&FloatFromCurve(
     FloatCurve=DataInterface() {propertyValues {Curve {keys [<at 0 1>,<at 0.0648 0.333>,<at 1 0>]}}}
     Scale Curve=&FloatFromCurve(FloatCurve=DataInterface() {propertyValues {Curve {keys [<at 0 1>,<at 0.4001 0.4154>,<at 1 0.4>]}}}, CurveIndex=$Emitter.NormalizedLoopAge)
    )
    Scale RGB=&VectorFromFloat(Value=&FloatFromCurve(FloatCurve=DataInterface() {propertyValues {Curve {keys [<at 0 10>,<at 0.15 3>,<at 1 1>]}}}))
    ScaleA=true
    ScaleRGB=true
   ) {module=ScaleColor001}
   &ParticleState()
   &ScaleMeshSize(Scale Factor=&VectorFromFloat(Value=&FloatFromCurve(
      FloatCurve=DataInterface() {propertyValues {Curve {keys [<at 0 0>,<at 0.1 1.4>,<at 1 2>]}}}
      Scale Curve=&Multiply_Float(A=1, B=1)
     )))
   &InheritSourceMovement(ApplyForce=false, ApplyPosition=true, ApplyPositionFalloff=false)
   &SolveForcesAndVelocity(Clamp Velocity=false, Limit Acceleration=false)
  ]
  rendererClasses [&NiagaraMeshRendererProperties]
  renderers [{rendererClass=&NiagaraMeshRendererProperties, rendererIndex=0}]
 }
 Spawn_Flash {
  emitterUpdateScript [
   &EmitterState(Inactive Response=::NewEnumerator1, Life Cycle Mode=::NewEnumerator1, Loop Behavior=::NewEnumerator1, Loop Duration=1, UseLoopDelay=true)
   &SpawnBurst_Instantaneous(Use Spawn Probability=false, UseLoopCountLimit=false)
  ]
  particleSpawnScript [
   &InitializeParticle(
    Color=<rgba 0.6054 0.8353 1 0.3>
    Lifetime=0.5
    Sprite Rotation Mode=::NewEnumerator2
    Sprite Size Mode=::NewEnumerator3
    Uniform Sprite Size=1000
    UsePositionOffset=false
    Write Material Random=true
   )
   &CameraOffset(Camera Offset Amount=50)
  ]
  particleUpdateScript [
   &ParticleState()
   &ScaleSpriteSize(Uniform Curve Sprite Scale=DataInterface() {propertyValues {Curve {keys [
        {arriveTangent=-0.4932, interpMode=RCIM_Cubic, leaveTangent=-0.4932, tangentMode=RCTM_User, time=0, value=0.55}
        {arriveTangent=-3.0216, interpMode=RCIM_Cubic, leaveTangent=-3.0216, tangentMode=RCTM_User, time=0.05, value=1.5}
        <at 1 0 RCIM_Cubic>
       ]}}})
   &ScaleColor(
    Scale Alpha=&Multiply_Float(
     A=&FloatFromCurve(
      FloatCurve=DataInterface() {propertyValues {Curve {keys [<at 0 0>,<at 0.1 1>,<at 0.4 1>,<at 1 0>]}}}
      Scale Curve=&FNTimeOfDayScalar(Night Scale=0.3, Morning Scale=0.5, Day Scale=0.5, Evening Scale=0.5)
     )
     B=&FN_FrontEndFloatModifier(Gameplay=1, Front End=0.3)
    )
    Scale RGB=&VectorFromFloat(Value=&FN_PickFloatDependOnPlatform(TargetPlatformActive=1, TargetPlatformNoActive=0.4))
    ScaleA=true
    ScaleRGB=true
   )
  ]
  rendererClasses [&NiagaraSpriteRendererProperties]
  renderers [{rendererClass=&NiagaraSpriteRendererProperties}]
 }
}
summary {
 systemName=NS_Emote_YuzuCrank_Ground_Spawn
 emitters [
  {emitterName=Spawn_Burst, rendererClasses [&NiagaraMeshRendererProperties]}
  {emitterName=Spawn_Flash, rendererClasses [&NiagaraSpriteRendererProperties]}
  {emitterName=Spawn_Board, rendererClasses [&NiagaraMeshRendererProperties, &NiagaraSpriteRendererProperties]}
 ]
 userVariables [
  User.bIsFrontEnd {type=NiagaraBool}
  User.bIsFrontEndPreview {type=NiagaraBool}
  User.DeltaTimeMax {type=NiagaraFloat, defaultValue=0.0167, description=""}
  User.DustSpawnCurve {type=NiagaraFloat, defaultValue=1, description=""}
  User.Fade {type=NiagaraFloat, defaultValue=0, description=""}
  User.RibbonIDOffset {type=NiagaraInt32}
  User.SkeletalMesh {
   type=NiagaraDataInterfaceSkeletalMesh
   defaultValue {
    struct=NiagaraExt_VariableValue_DataInterface
    value {dataInterface=&self:NiagaraDataInterfaceSkeletalMesh_0}
   }
  }
  User.SwingSectionIndex {type=NiagaraInt32}
  User.TrailLifeTimeMult {type=NiagaraFloat, defaultValue=0.4, description=""}
  User.WidthMult {type=NiagaraFloat, defaultValue=2, description=""}
 ]
}
