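# Niagara effect for the YuzuCrank emote's idle, built from four red-tinted emitters.
# Idle_Glow and Reactive_Glow burst glow sprites, the latter layering three renderers and copying User.ReactiveOn into ReactiveAlpha, though its ScaleColor consumer is disabled.
# Motion_Lines loops spinning, short-lived meshes tinted by time of day.
# Spawn_In_Out drives circuit, wave and glow sprite/mesh layers from one burst via separate normalized ages, with platform-dependent material parameters.
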NiagaraSystem /BRCosmetics/Animation/Game/MainPlayer/Emotes/YuzuCrank/FX/NS_Emote_YuzuCrank_Idle
dependencies {
 usedDataInterfaces [
  &NiagaraDataInterfaceCurve
  &NiagaraDataInterfaceSkeletalMesh
  &NiagaraDataInterfaceArrayFloat
  &NiagaraDataInterfaceColorCurve
  &NiagaraDataInterfacePlatformSet
 ]
 usedDynamicInputs [
  &MakeVector
  &RandomRangeFloat
  &FNTimeOfDayScalar
 ]
 usedModules [
  &SystemState
  &self:SystemSpawnScript.SystemScriptSource.SystemScriptGraph.NiagaraNodeAssignment_0.SetVariables_#1
 ]
}
emitters {
 Idle_Glow {
  emitterUpdateScript [
   &EmitterState()
   &SpawnBurst_Instantaneous(Use Spawn Probability=false, UseLoopCountLimit=false)
   &SpawnRate(SpawnRate=2, Use Spawn Probability=false) {enabled=false}
  ]
  particleSpawnScript [
   &InitializeParticle(
    Color=<rgba 0.9062 0.0943 0.0661 0.5>
    Lifetime=7
    Position Offset=<0 0.5 0>
    Position Offset Coordinate Space=::Local
    Sprite Rotation Mode=::NewEnumerator0
    Sprite Size Mode=::NewEnumerator3
    Uniform Sprite Size=39
    UsePositionOffset=true
    Write Material Random=true
   )
   &SpriteFacingAndAlignment(WriteAlignment=true, WriteFacing=true)
  ]
  particleUpdateScript [
   &ParticleState()
   &ScaleColor(
    Scale Alpha=&FloatFromCurve(FloatCurve=DataInterface() {propertyValues {Curve {keys [<at 0 0>,<at 0.1096 1>,<at 1 0>]}}})
    ScaleA=true
    ScaleRGB=true
   ) {enabled=false}
   &SkeletalMeshLocation(
    Bone Sampling Mode=::NewEnumerator4
    MaskSpawnGroup=false
    Skeletal Mesh=DataInterface() {propertyValues {
      DefaultMesh=null
      ExcludeBoneName=root
      OptionalComponentClass=null
      PreviewMesh=/BRCosmetics/Weapons/FORT_Melee/Pickaxe_YuzuCrankTide/Meshes/YuzuCrankTide_Axe
      SkinningMode=SkinOnTheFly
      SoftSourceActor=null
      SourceMode=Default
      UvSetIndex=0
      WholeMeshLOD=-1
      bExcludeBone=false
      bReadDeformedGeometry=true
      bRequireCurrentFrameData=true
      ComponentTags []
      FilteredBones [base_triangle]
      FilteredSockets []
      MeshUserParameter {parameter {name=None, typeDefHandle {registeredTypeIndex=1}, varData []}}
      SamplingRegions []
     }}
   ) {enabled=false}
  ]
  rendererClasses [&NiagaraSpriteRendererProperties]
  renderers [{rendererClass=&NiagaraSpriteRendererProperties}]
 }
 Motion_Lines {
  emitterUpdateScript [
   &EmitterState(Life Cycle Mode=::NewEnumerator1, Loop Behavior=::NewEnumerator1, Loop Delay=1.1, Loop Duration=1.6, UseLoopDelay=true)
   &SpawnRate(SpawnGroup=3, SpawnRate=&Multiply_Float(A=25, B=1), Use Spawn Probability=false) {module=SpawnRate003}
  ]
  particleSpawnScript [
   &InitializeParticle(
    Color=&RandomRangeLinearColor(Minimum=<rgba 3 0.7307 0.6619 0.1>, Maximum=<rgba 0.9896 0.3025 0.2113 0.2>)
    Lifetime=0.25
    Mesh Scale Max=<0.15 0.05 0.15>
    Mesh Scale Min=<0.15 0.05 0.15>
    Mesh Scale Mode=::NewEnumerator2
    UsePositionOffset=true
    Write Material Random=true
   )
   &CameraOffset(Camera Offset Amount=15)
   &InitialMeshOrientation(
    Mesh Axis To Orient=<0 1 0>
    Mesh Orientation Mode=::NewEnumerator2
    Rotation=&RandomRangeVector(Minimum=<0 0 0>, Maximum=<0 1 0>)
    Use Rotation Vector=true
   )
  ]
  particleUpdateScript [
   &ParticleState()
   &UpdateMeshOrientation(Rotation Rate=0.5, Rotation Vector=<0 3 0>) {module=UpdateMeshOrientation002}
   &ScaleMeshSize(Scale Factor=&VectorFromFloat(Value=&FloatFromCurve(FloatCurve=DataInterface() {propertyValues {Curve {keys [<at 0 0.7>,<at 0.45 0.875>,<at 1 1.5>]}}}, Scale Curve=1.25)))
   &DynamicMaterialParameters(
    Index 0 Param 1=&FloatFromCurve(FloatCurve=DataInterface() {propertyValues {Curve {keys [
         {arriveTangent=2.4535, interpMode=RCIM_Cubic, leaveTangent=2.4535, tangentMode=RCTM_User, time=0, value=0}
         {interpMode=RCIM_Cubic, tangentMode=RCTM_User, time=0.5, value=0.5}
         {arriveTangent=-2.5865, tangentMode=RCTM_User, time=1, value=0}
        ]}}})
    Module.Index0Param0Write=true
    Module.Index0Param1Write=false
    Module.Index0Param2Write=false
    Module.Index0Param3Write=false
   )
   &SolveForcesAndVelocity(Clamp Velocity=false, Limit Acceleration=false)
   &ScaleColor(
    Scale Alpha=&FloatFromCurve(
     FloatCurve=DataInterface() {propertyValues {Curve {keys [
         {arriveTangent=6.3511, interpMode=RCIM_Cubic, leaveTangent=6.3511, tangentMode=RCTM_User, time=0, value=0}
         {interpMode=RCIM_Cubic, tangentMode=RCTM_User, time=0.5, value=1}
         {tangentMode=RCTM_User, time=1, value=0}
        ]}}}
     Scale Curve=0.3
    )
    Scale RGB=&MakeVectorFromLinearColorRGB(Color=&FNTimeOfDayColor(
      Night Color=<rgba 0.1253 0.1914 0.5 1>
      Morning Color=<rgba 0.5 0.3383 0.3719 1>
      Day Color=<rgba 0.5 0.5 0.5 1>
      Evening Color=<rgba 0.4474 0.4547 0.4578 1>
     ))
    ScaleA=true
    ScaleRGB=true
   )
   &ScaleColor(Scale Alpha=0.4, ScaleA=true, ScaleRGB=true) {module=ScaleColor001}
  ]
  rendererClasses [&NiagaraMeshRendererProperties]
  renderers [{rendererClass=&NiagaraMeshRendererProperties, rendererIndex=0}]
 }
 Reactive_Glow {
  emitterUpdateScript [
   &EmitterState(Life Cycle Mode=::NewEnumerator1, Loop Behavior=::NewEnumerator1, Loop Delay=0.05, Loop Duration=1.5, UseLoopDelay=true)
   &SpawnBurst_Instantaneous(Use Spawn Probability=false, UseLoopCountLimit=false)
   &SpawnRate(Use Spawn Probability=false) {enabled=false}
  ]
  particleSpawnScript [
   &InitializeParticle(
    Color=<rgba 1 0 0.0548 1>
    Lifetime=3.5
    Position Offset=<0.25 2.8 0>
    Position Offset Coordinate Space=::Local
    Sprite Rotation Mode=::NewEnumerator0
    Sprite Size Mode=::NewEnumerator3
    Uniform Sprite Size=44
    UsePositionOffset=true
    Write Material Random=true
   )
   &SpriteFacingAndAlignment(WriteAlignment=true, WriteFacing=true)
   &CameraOffset(Camera Offset Amount=1)
  ]
  particleUpdateScript [
   &ParticleState()
   /SetVariables(Particles.ReactiveAlpha:NiagaraFloat=$User.ReactiveOn) {isSetParameters=true}
   &ScaleColor(
    Scale Alpha=&FloatFromCurve(FloatCurve=DataInterface() {propertyValues {Curve {keys [<at 0 0>,<at 0.1096 1>,<at 1 0>]}}}, Scale Curve=$Particles.ReactiveAlpha)
    ScaleA=true
    ScaleRGB=true
   ) {enabled=false}
   &SkeletalMeshLocation(
    Bone Sampling Mode=::NewEnumerator4
    MaskSpawnGroup=false
    Skeletal Mesh=DataInterface() {propertyValues {
      DefaultMesh=null
      ExcludeBoneName=root
      OptionalComponentClass=null
      PreviewMesh=&YuzuCrankTide_Axe
      SkinningMode=SkinOnTheFly
      SoftSourceActor=null
      SourceMode=Default
      UvSetIndex=0
      WholeMeshLOD=-1
      bExcludeBone=false
      bReadDeformedGeometry=true
      bRequireCurrentFrameData=true
      ComponentTags []
      FilteredBones [base_triangle]
      FilteredSockets []
      MeshUserParameter {parameter {name=None, typeDefHandle {registeredTypeIndex=1}, varData []}}
      SamplingRegions []
     }}
   ) {enabled=false}
   /SetVariables(
    Particles.AmbientSize:Vector2f=&Multiply_Vector2D(A=$Particles.SpriteSize, B=&Vector2DFromFloat(Value=1.1))
    Particles.BloomColor:LinearColor=<rgba 15 2.5394 0 1>
    Particles.BloomSize:Vector2f=&Multiply_Vector2D(A=$Particles.SpriteSize, B=&Vector2DFromFloat(Value=1.05))
    Particles.TightColor:LinearColor=<rgba 50 7.6515 1.8229 1>
   ) {isSetParameters=true}
  ]
  rendererClasses [&NiagaraSpriteRendererProperties]
  renderers [
   {rendererClass=&NiagaraSpriteRendererProperties}
   {rendererClass=&NiagaraSpriteRendererProperties, rendererIndex=1}
   {rendererClass=&NiagaraSpriteRendererProperties, rendererIndex=2}
  ]
 }
 Spawn_In_Out {
  emitterSpawnScript [/SetVariables(Emitter.Size=0.3) {isSetParameters=true}]
  emitterUpdateScript [
   &EmitterState(Life Cycle Mode=::NewEnumerator1, Loop Behavior=::NewEnumerator1, Loop Duration=15, UseLoopDelay=false)
   &SpawnBurst_Instantaneous(Use Spawn Probability=false, UseLoopCountLimit=false) {module=SpawnBurst_Instantaneous001}
  ]
  particleSpawnScript [
   &InitializeParticle(
    Color=<rgba 4 4 4 1>
    Lifetime=0.8
    Mesh Scale Mode=::NewEnumerator3
    Mesh Uniform Scale=&Multiply_Float(A=0.5, B=$Emitter.Size)
    Position Offset=<0 0 -2>
    Sprite Rotation Angle=&SelectFloatFromArray(
     Array Sampling Mode=::NewEnumerator0
     Direct Array Index=?
     Float Selection Array=DataInterface() {propertyValues {FloatData [0,0.25,0.5,0.75]}}
     Random Seed=0
    )
    Sprite Rotation Mode=::NewEnumerator1
    Sprite Size Mode=::NewEnumerator3
    Uniform Sprite Size=&Multiply_Float(A=200, B=$Emitter.Size)
    UsePositionOffset=true
    Write Material Random=true
   )
   /SetVariables(Particles.VisibilityTag=$Emitter.SpawnGroup) {isSetParameters=true}
   /SetVariables(Particles.WaveColor:LinearColor=<rgba 5 0 0 1>, Particles.WaveLifetime=0.6) {isSetParameters=true}
   /SetVariables(
    Particles.Glow Lifetime=0.2
    Particles.Glow Size=&Multiply_Vector2D_ByFloat(Float=$Emitter.Size, Vector2D=&Multiply_Vector2D_ByFloat(Float=80, Vector2D=<1 1>))
   ) {isSetParameters=true}
   &InitialMeshOrientation(
    Mesh Orientation Mode=::NewEnumerator2
    Orientation Vector=<0 0 1>
    Rotation=<0 0 0.5>
    Rotation Coordinate Space=::NewEnumerator0
    Use Rotation Vector=false
   )
   /SetVariables(Particles.Circuit Lifetime=0.4) {isSetParameters=true}
  ]
  particleUpdateScript [
   &ParticleState()
   /SetVariables(Particles.Update Circuit Age:NiagaraFloat=&Divide_Float(A=$Particles.Age, B=$Particles.Circuit Lifetime)) {isSetParameters=true}
   /SetVariables(Particles.Update WaveAge:NiagaraFloat=&Divide_Float(A=$Particles.Age, B=$Particles.WaveLifetime)) {isSetParameters=true}
   /SetVariables(Particles.Update Glow Age:NiagaraFloat=&Divide_Float(A=$Particles.Age, B=$Particles.Glow Lifetime)) {isSetParameters=true}
   &ScaleColor(
    Scale Mode=::NewEnumerator1
    Scale RGBA=&Vector4FromCurve(
     CurveIndex=$Particles.Update Circuit Age
     Vector4Curve=DataInterface() {propertyValues {
       AlphaCurve {keys [<at 0.3174 1>,<at 0.5961 0>]}
       BlueCurve {keys [<at 0 0>,<at 1 0>]}
       GreenCurve {keys [<at 0 0>,<at 1 0>]}
       RedCurve {keys [<at 0 1>,<at 1 0.2>]}
      }}
    )
    ScaleRGBA=true
   )
   &ScaleSpriteSize(Scale Factor=&Vector2DFromFloat(Value=&FloatFromCurve(
      FloatCurve=DataInterface() {propertyValues {Curve {keys [
          {arriveTangent=0.5497, interpMode=RCIM_Cubic, leaveTangent=0.5497, tangentMode=RCTM_User, time=0, value=0.4128}
          {arriveTangent=0.5872, interpMode=RCIM_Cubic, leaveTangent=0.5872, time=0.2754, value=0.8613}
          {interpMode=RCIM_Cubic, tangentMode=RCTM_User, time=1, value=1}
         ]}}}
      CurveIndex=$Particles.Update Circuit Age
     )))
   &DynamicMaterialParameters(
    Index 0 Param 2=&FloatFromCurve(
     FloatCurve=DataInterface() {propertyValues {Curve {keys [
         <at 0 1 RCIM_Cubic>
         {arriveTangent=0.1115, interpMode=RCIM_Cubic, leaveTangent=0.1115, tangentMode=RCTM_User, time=0.1, value=0.5}
         {interpMode=RCIM_Cubic, tangentMode=RCTM_User, time=1, value=1}
        ]}}}
     CurveIndex=$Particles.Update Circuit Age
    )
    Module.Index0Float0Write=true
    Module.Index0Float1Write=true
    Module.Index0Float2Write=true
    Module.Index0Float3Write=true
   )
   /SetVariables(Particles.DY0:Vector4f=&MakeVector4(
     X=0
     Y=&FloatFromCurve(FloatCurve=DataInterface() {propertyValues {Curve {keys [{tangentMode=RCTM_User, time=0, value=0.47}, {tangentMode=RCTM_User, time=1, value=1}]}}})
     Z=0
     W=0
    )) {isSetParameters=true}
   &ScaleMeshSize(Scale Factor=&VectorFromFloat(Value=&FloatFromCurve(
      FloatCurve=DataInterface() {propertyValues {Curve {keys [
          {arriveTangent=4.8748, interpMode=RCIM_Cubic, leaveTangent=4.8748, tangentMode=RCTM_User, time=0, value=1}
          {arriveTangent=0.9672, interpMode=RCIM_Cubic, leaveTangent=0.9672, tangentMode=RCTM_User, time=0.3099, value=1.9838}
          {arriveTangent=0.1926, interpMode=RCIM_Cubic, leaveTangent=0.1926, tangentMode=RCTM_User, time=1, value=2.1883}
         ]}}}
      CurveIndex=$Particles.Update WaveAge
     )))
   /SetVariables(
    Particles.Update Wave Color:LinearColor=&MakeLinearColorFromVectorAndFloat(
     "Vector (RGB)"=&Multiply_VectorByFloat(
      Vector=&MakeVectorFromLinearColorRGB(Color=$Particles.WaveColor)
      Float=&FloatFromCurve(
       FloatCurve=DataInterface() {propertyValues {Curve {
          defaultValue=1
          keys [
           {tangentMode=RCTM_User, time=0, value=1}
           {arriveTangent=-0.0405, interpMode=RCIM_Cubic, leaveTangent=-0.0405, tangentMode=RCTM_User, time=0.7009, value=1}
           {arriveTangent=-8.0089, interpMode=RCIM_Cubic, leaveTangent=-8.0089, tangentMode=RCTM_User, time=1, value=0}
          ]
         }}}
       Scale Curve=15
      )
     )
     "Float (Alpha)"=&FloatFromCurve(
      FloatCurve=DataInterface() {propertyValues {Curve {keys [
          <at 0 0 RCIM_Cubic>
          {arriveTangent=0.4167, interpMode=RCIM_Cubic, leaveTangent=0.4167, time=0.1, value=1}
          {arriveTangent=-1.1111, interpMode=RCIM_Cubic, leaveTangent=-1.1111, time=0.6, value=0.25}
          <at 1 0 RCIM_Cubic>
         ]}}}
      CurveIndex=$Particles.Update WaveAge
      Scale Curve=0.125
     )
    )
    Particles.Update Wave DynamicParameter:Vector4f=&MakeVector4(
     X=&FloatFromCurve(FloatCurve=DataInterface() {propertyValues {Curve {keys [<at 0 1>,<at 1 0>]}}}, CurveIndex=$Particles.Update WaveAge)
     Y=2
     Z=&FloatFromCurve(FloatCurve=DataInterface() {propertyValues {Curve {keys [<at 0.1 1>,<at 1 0>]}}}, CurveIndex=$Particles.Update WaveAge, Scale Curve=6)
     W=&FloatFromCurve(
      FloatCurve=DataInterface() {propertyValues {Curve {keys [
          <at 0 1>
          {arriveTangent=0.3079, interpMode=RCIM_Cubic, leaveTangent=0.3079, tangentMode=RCTM_User, time=0.3, value=0.8}
          <at 1 1 RCIM_Cubic>
         ]}}}
      CurveIndex=$Particles.Update WaveAge
     )
    )
    Particles.Update Wave DynamicParameter001:Vector4f=&MakeVector4(
     X=&FN_PickFloatDependOnPlatform(TargetPlatformActive=180, TargetPlatformNoActive=120)
     Y=&FN_PickFloatDependOnPlatform(TargetPlatformActive=2.5, TargetPlatformNoActive=4)
     Z=0
     W=0
    )
   ) {isSetParameters=true}
   /SetVariables(
    Particles.Update Glow Color:LinearColor=&MakeLinearColorFromVectorAndFloat(
     "Vector (RGB)"=&MakeVectorFromLinearColorRGB(Color=<rgba 1 0.0833 0.0833 1>)
     "Float (Alpha)"=&FloatFromCurve(
      FloatCurve=DataInterface() {propertyValues {Curve {
         defaultValue=1
         keys [
          {interpMode=RCIM_Cubic, tangentMode=RCTM_User, time=0, value=0}
          {interpMode=RCIM_Cubic, tangentMode=RCTM_User, time=0.2399, value=1}
          {tangentMode=RCTM_User, time=1, value=0}
         ]
        }}}
      CurveIndex=$Particles.Update Glow Age
     )
    )
    Particles.Update Glow DynamicParameter:Vector4f=<2 0 0 0>
    Particles.Update Glow Size:Vector2f=&Multiply_Vector2D(
     A=$Particles.Glow Size
     B=&Multiply_Vector2D_ByFloat(
      Float=&FloatFromCurve(
       FloatCurve=DataInterface() {propertyValues {Curve {
          defaultValue=0.367
          keys [
           {interpMode=RCIM_Cubic, tangentMode=RCTM_User, time=0, value=0.367}
           {arriveTangent=-0.1298, interpMode=RCIM_Cubic, leaveTangent=-0.1298, tangentMode=RCTM_User, time=1, value=3.8846}
          ]
         }}}
       CurveIndex=$Particles.Update Glow Age
      )
      Vector2D=<1.25 1.25>
     )
    )
   ) {isSetParameters=true}
   /SetVariables(
    Particles.WaveColor2:LinearColor=&MultiplyLinearColor(
     Linear Color=&ColorFromCurve(ColorCurve=DataInterface() {propertyValues {
        AlphaCurve {keys [<at 0 1>,<at 1 0>]}
        BlueCurve {keys [<at 0 10>,<at 0.15 3>,<at 1 1>]}
        GreenCurve {keys [<at 0 10>,<at 0.15 3>,<at 1 1>]}
        RedCurve {keys [<at 0 10>,<at 0.15 3>,<at 1 1>]}
       }})
     Scale Factor=&MakeVector4_FromLinearColor(LinearColor=<rgba 10 0.7148 0 0.05>)
    )
    Particles.WaveDynamicMaterialParameter:Vector4f=&MakeVector4(
     X=&FloatFromCurve(FloatCurve=DataInterface() {propertyValues {Curve {keys [{arriveTangent=-2.1015, interpMode=RCIM_Cubic, leaveTangent=-2.1015, tangentMode=RCTM_User, time=0, value=0.5}, <at 1 -0.15>]}}})
     Y=&FloatFromCurve(FloatCurve=DataInterface() {propertyValues {Curve {keys [{arriveTangent=-0.3464, interpMode=RCIM_Cubic, leaveTangent=-0.3464, tangentMode=RCTM_User, time=0, value=1}, <at 1 0.9>]}}})
     Z=&FloatFromCurve(FloatCurve=DataInterface() {propertyValues {Curve {keys [<at 0 0>,<at 0.15 1>,<at 1 1.2>]}}}, Scale Curve=0.8)
     W=0
    )
    Particles.WaveScale:Vector3f=&VectorFromFloat(Value=&FloatFromCurve(
      FloatCurve=DataInterface() {propertyValues {Curve {keys [{arriveTangent=2.8792, interpMode=RCIM_Cubic, leaveTangent=2.8792, tangentMode=RCTM_User, time=0, value=1}, <at 1 2.2 RCIM_Cubic>]}}}
      Scale Curve=0.35
     ))
   ) {isSetParameters=true}
  ]
  rendererClasses [&NiagaraSpriteRendererProperties, &NiagaraMeshRendererProperties]
  renderers [
   {rendererClass=&NiagaraSpriteRendererProperties}
   {rendererClass=&NiagaraMeshRendererProperties, rendererIndex=1}
   {rendererClass=&NiagaraSpriteRendererProperties, rendererIndex=2}
   {rendererClass=&NiagaraMeshRendererProperties, rendererIndex=3}
  ]
 }
}
summary {
 systemName=NS_Emote_YuzuCrank_Idle
 emitters [
  {emitterName=Idle_Glow, rendererClasses [&NiagaraSpriteRendererProperties]}
  {emitterName=Motion_Lines, rendererClasses [&NiagaraMeshRendererProperties]}
  {emitterName=Reactive_Glow, rendererClasses [&NiagaraSpriteRendererProperties]}
  {emitterName=Spawn_In_Out, rendererClasses [&NiagaraSpriteRendererProperties, &NiagaraMeshRendererProperties]}
 ]
 userVariables [
  User.ReactiveBool {type=NiagaraBool}
  User.ReactiveOn {type=NiagaraFloat, defaultValue=1, description=""}
  User.ReactivityOn {type=NiagaraFloat, defaultValue=1, description=""}
  User.Timestamp {type=NiagaraFloat, defaultValue=1, description=""}
 ]
}
