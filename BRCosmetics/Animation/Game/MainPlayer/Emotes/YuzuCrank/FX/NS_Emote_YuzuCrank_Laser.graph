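# Niagara system for the laser effect of the YuzuCrank emote, a cyan-blue beam with a glowing core and an opening flash.
# Trail_Bright draws the beam as four mesh layers that stretch out, then fade via curve-driven color and material parameters.
# Core_Glow stacks sprites, a mesh and a flickering light; Spawn_Flash adds a brief burst sprite.
# Brightness scales by time of day, front-end preview and platform; mesh sizes compensate for owner scale.
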NiagaraSystem /BRCosmetics/Animation/Game/MainPlayer/Emotes/YuzuCrank/FX/NS_Emote_YuzuCrank_Laser
dependencies {
 usedDataInterfaces [
  &NiagaraDataInterfaceCurve
  &NiagaraDataInterfacePlatformSet
  &NiagaraDataInterfaceVectorCurve
  &NiagaraDataInterfaceVector2DCurve
  &NiagaraDataInterfaceColorCurve
 ]
 usedDynamicInputs []
 usedModules [&SystemState]
}
emitters {
 Core_Glow {
  emitterUpdateScript [
   &EmitterState(Inactive Response=::NewEnumerator1, Life Cycle Mode=::NewEnumerator1, Loop Behavior=::NewEnumerator1, Loop Duration=4.5, UseLoopDelay=false)
   &SpawnBurst_Instantaneous(Use Spawn Probability=false, UseLoopCountLimit=false)
  ]
  particleSpawnScript [
   &InitializeParticle(
    Color=<rgba 0.5309 0.6232 0.9 1.5>
    Lifetime=4.4
    Mesh Scale=<0.05 0.08 0.05>
    Mesh Scale Mode=::NewEnumerator4
    Position Offset=<0 0 1>
    Sprite Size=<180 150>
    Sprite Size Mode=::NewEnumerator4
    UsePositionOffset=true
    Write Material Random=true
   )
   &InitialMeshOrientation(Mesh Orientation Mode=::NewEnumerator2, Orientation Vector=<0 0 1>, Rotation=<0 0.25 0>, Use Rotation Vector=true)
   &CameraOffset(Camera Offset Amount=5)
   &AddVelocity(Velocity=<0 0 0.001>)
   /SetVariables() {isSetParameters=true}
  ]
  particleUpdateScript [
   &ParticleState()
   &ScaleMeshSize(Scale Factor=&Multiply_Vector(
     A=&ScaleFloatByCurve(
      InitialValue:Vector3f=<1 1 1>
      ScaleCurve:NiagaraDataInterfaceVectorCurve=DataInterface() {propertyValues {
        XCurve {keys [<at 0 0.25>,<at 0.1 1>,<at 0.9 1>,<at 1 0>]}
        YCurve {keys [<at 0 1.25>,<at 0.1 2.25>,<at 0.9 2.25>,<at 1 0>]}
        ZCurve {keys [<at 0 1.25>,<at 0.1 2.25>,<at 0.9 2.25>,<at 1 0>]}
       }}
      CurveIndex:NiagaraFloat=$Particles.NormalizedAge
     )
     B=&Divide_Vector(A:Vector3f=<1 1 1>, B:Vector3f=$Engine.Owner.Scale)
    ))
   &ScaleSpriteSize(
    Uniform Curve Scale=0.8
    Uniform Curve Sprite Scale=DataInterface() {propertyValues {Curve {keys [
        <at 0 0>
        <at 0.1 1>
        <at 0.1512 0.7174>
        <at 0.2132 1>
        <at 0.2641 0.7539>
        <at 0.299 1>
        <at 0.3563 0.7037>
        <at 0.4151 1>
        <at 0.4564 0.8017>
        <at 0.5041 1>
        <at 0.5502 0.7561>
        <at 0.5836 1>
        <at 0.6551 0.7653>
        <at 0.6981 1>
        <at 0.7521 0.7721>
        <at 0.7871 1>
        <at 0.8475 0.7106>
        <at 0.9 1>
        {tangentMode=RCTM_User, time=1.0046, value=1}
       ]}}}
   )
   /SetVariables(Particles.Size002:Vector2f=&Multiply_Vector2D(
     A=$Particles.SpriteSize
     B=&Vector2DFromCurve(
      Vector2Curve=DataInterface() {propertyValues {XCurve {keys [<at 0 1>,<at 0.3 1>,<at 0.65 0>]}, YCurve {keys [<at 0 1>,<at 1 1>]}}}
      ScaleCurve=<0.7 0.8>
     )
    )) {isSetParameters=true}
   /SetVariables(Particles.Size003:Vector2f=&Multiply_Vector2D(A=<150 150>, B=<1 3>)) {isSetParameters=true}
   &ScaleColor(
    Scale Alpha=&Multiply_Float(
     A=&FloatFromCurve(
      FloatCurve=DataInterface() {propertyValues {Curve {keys [
          {arriveTangent=4.4475, time=0, value=0}
          {arriveTangent=5.3518, leaveTangent=-3.1774, time=0.1, value=1.2}
          {arriveTangent=-1.4972, leaveTangent=-1.4972, time=0.1523, value=0.7}
          {arriveTangent=-0.1667, leaveTangent=-0.1667, time=0.2052, value=1.1711}
          <at 0.258 0.7713>
          <at 0.2926 1.1775>
          {arriveTangent=-0.25, leaveTangent=-0.25, time=0.3608, value=0.689}
          <at 0.6036 0>
          <at 1 0>
         ]}}}
      Scale Curve=&FNTimeOfDayScalar(Night Scale=0.6, Morning Scale=0.7, Day Scale=0.8, Evening Scale=0.6)
     )
     B=&FN_FrontEndFloatModifier(Gameplay=1, Front End=0.3)
    )
    Scale RGB=&MakeVectorFromLinearColorRGB(Color=&ColorFromCurve(ColorCurve=DataInterface() {propertyValues {
        AlphaCurve {keys []}
        BlueCurve {keys [<at 0.7887 0.4375 RCIM_Cubic>,<at 0.9801 0.1562 RCIM_Cubic>]}
        GreenCurve {keys [<at 0.7887 0.3537 RCIM_Cubic>,<at 0.9801 0.0486 RCIM_Cubic>]}
        RedCurve {keys [<at 0.7887 0 RCIM_Cubic>,<at 0.9801 0 RCIM_Cubic>]}
       }}))
    ScaleA=true
    ScaleRGB=true
   )
   /SetVariables(Particles.color002:LinearColor=&MakeLinearColorFromVectorAndFloat(
     "Vector (RGB)"=&MakeVectorFromLinearColorRGB(Color=<rgba 0.2518 0.5919 0.7 1>)
     "Float (Alpha)"=&FloatFromCurve(
      FloatCurve=DataInterface() {propertyValues {Curve {keys [<at 0 1>,<at 0.1 2>,<at 0.61 0>,<at 1 0>]}}}
      Scale Curve=&FN_PickFloatDependOnPlatform(TargetPlatformActive=5, TargetPlatformNoActive=1)
     )
    )) {isSetParameters=true}
   /SetVariables(Particles.Color003:LinearColor=&MakeLinearColorFromVectorAndFloat(
     "Vector (RGB)"=&MakeVectorFromLinearColorRGB(Color=<rgba 0.3028 0.2438 0.4375 1>)
     "Float (Alpha)"=&Multiply_Float(
      A=&FloatFromCurve(
       FloatCurve=DataInterface() {propertyValues {Curve {keys [<at 0 1>,<at 0.1 2>,<at 0.46 0>,<at 1 0>]}}}
       Scale Curve=&FNTimeOfDayScalar(Night Scale=0.2, Morning Scale=0.1, Day Scale=0.4, Evening Scale=0.2)
      )
      B=&Multiply_Float(
       A=&FN_FrontEndFloatModifier(Gameplay=0.5, Front End=0.15)
       B=&FN_PickFloatDependOnPlatform(TargetPlatformActive=1, TargetPlatformNoActive=0.4)
      )
     )
    )) {isSetParameters=true}
   /SetVariables(Particles.MeshColor:LinearColor=&MakeLinearColorFromVectorAndFloat(
     "Vector (RGB)"=&MakeVectorFromLinearColorRGB(Color=&ColorFromCurve(ColorCurve=DataInterface() {propertyValues {
         AlphaCurve {keys []}
         BlueCurve {keys [
           {arriveTangent=-1.8762, interpMode=RCIM_Cubic, leaveTangent=-1.8762, tangentMode=RCTM_User, time=0.1298, value=2}
           {interpMode=RCIM_Cubic, leaveTangent=0.2132, time=0.3441, value=2}
          ]}
         GreenCurve {keys [
           {arriveTangent=-0.4331, interpMode=RCIM_Cubic, leaveTangent=-0.4331, tangentMode=RCTM_User, time=0.1298, value=1.0965}
           {interpMode=RCIM_Cubic, leaveTangent=2.0123, time=0.3441, value=1.6429}
          ]}
         RedCurve {keys [
           {arriveTangent=-4.577, interpMode=RCIM_Cubic, leaveTangent=-4.577, tangentMode=RCTM_User, time=0.1298, value=0.7143}
           {interpMode=RCIM_Cubic, leaveTangent=-2.7231, time=0.3441, value=1.2129}
          ]}
        }}))
     "Float (Alpha)"=&Multiply_Float(
      A=&FloatFromCurve(
       FloatCurve=DataInterface() {propertyValues {Curve {keys [
           {arriveTangent=-0.9, time=0, value=0}
           {arriveTangent=-0.0998, leaveTangent=-0.0998, time=0.05, value=1}
           {arriveTangent=-1.0872, leaveTangent=-1.0872, time=0.1804, value=0.1}
           <at 0.5 0>
           {arriveTangent=-1.5912, leaveTangent=-1.5912, time=1, value=0}
          ]}}}
       Scale Curve=&FN_FrontEndFloatModifier(Gameplay=1, Front End=0.5)
      )
      B=&FNTimeOfDayScalar(Night Scale=0.7, Morning Scale=0.5, Evening Scale=0.6)
     )
    )) {isSetParameters=true}
   &FN_Light_Attributes(
    Color=&ColorFromCurve(ColorCurve=DataInterface() {propertyValues {
       AlphaCurve {keys []}
       BlueCurve {keys [<at 0 0 RCIM_Cubic>,<at 0.0858 30>,<at 0.1804 10>,<at 0.9 2>,<at 1 0 RCIM_Cubic>]}
       GreenCurve {keys [<at 0 0 RCIM_Cubic>,<at 0.0858 20.9816>,<at 0.1804 7.7416>,<at 0.9 1.2674>,<at 1 0 RCIM_Cubic>]}
       RedCurve {keys [<at 0 0 RCIM_Cubic>,<at 0.0858 0>,<at 0.1804 0>,<at 0.9 0>,<at 1 0 RCIM_Cubic>]}
      }})
    Create Color Attribute=true
    Create Exponent Attribute=true
    Create Position Attribute=true
    Create Radius Attribute=true
    Exponent=3
    Position=$Particles.SkeletalMeshLocation.SampledPosition
    Radius=&FloatFromCurve(
     FloatCurve=DataInterface() {propertyValues {Curve {keys [<at 0 0.4>,<at 0.1 1.5>,<at 0.2 1>,<at 0.5 1>,<at 0.8 0.8>,<at 1 0.4>]}}}
     Scale Curve=&Multiply_Float(A=&MakeCustomFloatFromBool(Bool=$User.bIsFrontEndPreview, True Float=40, False Float=80), B=1)
    )
    Scale Color=&FloatFromCurve(
     FloatCurve=DataInterface() {propertyValues {Curve {keys [<at 0 0.5>,<at 0.1 1.1>,<at 0.1401 0.4933>,<at 0.2 1.3742>,<at 0.2354 0.5667>,<at 0.3 1.1>,<at 0.3404 0.4667>,<at 0.4098 1.3991>,<at 0.4342 0.48>,<at 0.5 1.1>,<at 0.5311 0.5133>,<at 0.5829 1.3306>,<at 0.6376 0.5133>,<at 0.7 1.1>,<at 0.741 0.5067>,<at 0.7989 1.4926>,<at 0.8459 0.3133>,<at 0.9 1.1>,<at 1 0>]}}}
     Scale Curve=&Multiply_Float(A=5, B=&FNTimeOfDayScalar(Night Scale=0.3, Morning Scale=0.5, Day Scale=0.8, Evening Scale=0.4))
    )
   )
   &SolveForcesAndVelocity(Clamp Velocity=false, Limit Acceleration=false)
  ]
  rendererClasses [
   &NiagaraSpriteRendererProperties
   &NiagaraMeshRendererProperties
   &NiagaraLightRendererProperties
  ]
  renderers [
   {rendererClass=&NiagaraSpriteRendererProperties}
   {rendererClass=&NiagaraSpriteRendererProperties, rendererIndex=1}
   {rendererClass=&NiagaraSpriteRendererProperties, rendererIndex=2}
   {rendererClass=&NiagaraMeshRendererProperties, rendererIndex=3}
   {rendererClass=&NiagaraLightRendererProperties, rendererIndex=4}
  ]
 }
 Spawn_Flash {
  emitterUpdateScript [
   &EmitterState(Inactive Response=::NewEnumerator1, Life Cycle Mode=::NewEnumerator1, Loop Behavior=::NewEnumerator1, Loop Duration=1, UseLoopDelay=false)
   &SpawnBurst_Instantaneous(Use Spawn Probability=false, UseLoopCountLimit=false)
  ]
  particleSpawnScript [
   &InitializeParticle(
    Color=<rgba 0.6054 0.8353 1 0.3>
    Lifetime=0.5
    Sprite Rotation Mode=::NewEnumerator2
    Sprite Size Mode=::NewEnumerator3
    Uniform Sprite Size=350
    UsePositionOffset=false
    Write Material Random=true
   )
   &CameraOffset(Camera Offset Amount=50)
  ]
  particleUpdateScript [
   &ParticleState()
   &ScaleSpriteSize(Uniform Curve Sprite Scale=DataInterface() {propertyValues {Curve {keys [
        {arriveTangent=-0.4932, interpMode=RCIM_Cubic, leaveTangent=-0.4932, tangentMode=RCTM_User, time=0, value=0.55}
        {arriveTangent=-3.0216, interpMode=RCIM_Cubic, leaveTangent=-3.0216, tangentMode=RCTM_User, time=0.05, value=1.5}
        <at 1 0 RCIM_Cubic>
       ]}}})
   &ScaleColor(
    Scale Alpha=&Multiply_Float(
     A=&FloatFromCurve(
      FloatCurve=DataInterface() {propertyValues {Curve {keys [<at 0 0>,<at 0.1 1>,<at 0.4 1>,<at 1 0>]}}}
      Scale Curve=&FNTimeOfDayScalar(Night Scale=0.3, Morning Scale=0.6, Evening Scale=0.3)
     )
     B=&FN_FrontEndFloatModifier(Gameplay=1, Front End=0.3)
    )
    Scale RGB=&VectorFromFloat(Value=&FN_PickFloatDependOnPlatform(TargetPlatformActive=1, TargetPlatformNoActive=0.4))
    ScaleA=true
    ScaleRGB=true
   )
  ]
  rendererClasses [&NiagaraSpriteRendererProperties]
  renderers [{rendererClass=&NiagaraSpriteRendererProperties}]
 }
 Trail_Bright {
  emitterUpdateScript [
   &EmitterState(Inactive Response=::NewEnumerator1, Life Cycle Mode=::NewEnumerator1, Loop Behavior=::NewEnumerator1, Loop Duration=3.1, UseLoopDelay=false)
   &SpawnBurst_Instantaneous(Use Spawn Probability=false, UseLoopCountLimit=false)
   /SetVariables(Emitter.SplineWidth_Base:NiagaraFloat=&FloatFromCurve(
     FloatCurve=DataInterface() {propertyValues {Curve {keys [
         <at 0 3>
         <at 0.1 2>
         {arriveTangent=-6.7804, interpMode=RCIM_Cubic, leaveTangent=-6.7804, tangentMode=RCTM_User, time=0.7, value=2}
         {arriveTangent=-0.8494, interpMode=RCIM_Cubic, leaveTangent=-0.8494, tangentMode=RCTM_User, time=1, value=0}
        ]}}}
     CurveIndex=$Emitter.NormalizedLoopAge
     Scale Curve=10
    )) {isSetParameters=true}
   /SetVariables(Emitter.SplineWidth_Tip:NiagaraFloat=&FloatFromCurve(
     FloatCurve=DataInterface() {propertyValues {Curve {keys [
         <at 0 1>
         {arriveTangent=-6.7804, interpMode=RCIM_Cubic, leaveTangent=-6.7804, tangentMode=RCTM_User, time=0.7, value=1}
         {arriveTangent=-0.8494, interpMode=RCIM_Cubic, leaveTangent=-0.8494, tangentMode=RCTM_User, time=1, value=0}
        ]}}}
     CurveIndex=$Emitter.NormalizedLoopAge
     Scale Curve=10
    )) {isSetParameters=true}
   /SetVariables(Emitter.SplineWidthRed_Base:NiagaraFloat=&FloatFromCurve(
     FloatCurve=DataInterface() {propertyValues {Curve {keys [
         <at 0 1>
         <at 0.1 1>
         {arriveTangent=-4.6776, interpMode=RCIM_Cubic, leaveTangent=-4.6776, tangentMode=RCTM_User, time=0.6, value=1}
         {arriveTangent=-0.5343, interpMode=RCIM_Cubic, leaveTangent=-0.5343, tangentMode=RCTM_User, time=1, value=0}
        ]}}}
     CurveIndex=$Emitter.NormalizedLoopAge
     Scale Curve=10
    )) {isSetParameters=true}
   /SetVariables(Emitter.SplineWidthRed_Tip:NiagaraFloat=&FloatFromCurve(
     FloatCurve=DataInterface() {propertyValues {Curve {keys [
         <at 0 1>
         {arriveTangent=-4.6776, interpMode=RCIM_Cubic, leaveTangent=-4.6776, tangentMode=RCTM_User, time=0.6, value=1}
         {arriveTangent=-0.5343, interpMode=RCIM_Cubic, leaveTangent=-0.5343, tangentMode=RCTM_User, time=1, value=0}
        ]}}}
     CurveIndex=$Emitter.NormalizedLoopAge
     Scale Curve=10
    )) {isSetParameters=true}
  ]
  particleSpawnScript [
   &InitializeParticle(
    Color=<rgba 1 1 1 0.3>
    Lifetime=3.1
    Mesh Scale=<1 1 3>
    Mesh Scale Mode=::NewEnumerator4
    UsePositionOffset=false
    Write Material Random=true
   )
   &InitialMeshOrientation(Mesh Orientation Mode=::NewEnumerator2, Use Rotation Vector=true)
  ]
  particleUpdateScript [
   &ParticleState()
   &ScaleMeshSize(Scale Factor=&Multiply_Vector(
     A=&MakeVector(
      X=1
      Y=1
      Z=&FloatFromCurve(FloatCurve=DataInterface() {propertyValues {Curve {keys [<at 0 0>,<at 0.05 1>,<at 1 1>]}}})
     )
     B=&Divide_Vector(A:Vector3f=<1 1 1>, B:Vector3f=$Engine.Owner.Scale)
    ))
   &ScaleColor(
    Scale Alpha=&Multiply_Float(
     A=&FloatFromCurve(
      FloatCurve=DataInterface() {propertyValues {Curve {keys [
          <at 0 1>
          {arriveTangent=-0.2775, interpMode=RCIM_Cubic, leaveTangent=-0.2775, tangentMode=RCTM_User, time=0.8, value=1}
          {arriveTangent=-2.6012, interpMode=RCIM_Cubic, leaveTangent=-2.6012, tangentMode=RCTM_User, time=0.9, value=0.85}
          {arriveTangent=-14.0132, interpMode=RCIM_Cubic, leaveTangent=-14.0132, tangentMode=RCTM_User, time=1, value=0}
         ]}}}
      Scale Curve=&FNTimeOfDayScalar(Night Scale=0.6, Morning Scale=0.8, Evening Scale=0.6)
     )
     B=&FN_FrontEndFloatModifier(Gameplay=1, Front End=0.7)
    )
    ScaleA=true
    ScaleRGB=true
   )
   /SetVariables(Particles.color002:LinearColor=&MakeLinearColorFromVectorAndFloat(
     "Vector (RGB)"=&MakeVectorFromLinearColorRGB(Color=<rgba 0 41.6591 50 1>)
     "Float (Alpha)"=&Multiply_Float(
      A=&FloatFromCurve(
       FloatCurve=DataInterface() {propertyValues {Curve {keys [<at 0 1>,<at 0.95 1>,<at 1 0>]}}}
       Scale Curve=&FNTimeOfDayScalar(Night Scale=0.3, Morning Scale=0.3, Day Scale=0.3, Evening Scale=0.3)
      )
      B=&Multiply_Float(
       A=&FN_FrontEndFloatModifier(Gameplay=1, Front End=0.8)
       B=&FN_PickFloatDependOnPlatform(TargetPlatformActive=1, TargetPlatformNoActive=0.8)
      )
     )
    )) {isSetParameters=true}
   /SetVariables(Particles.Color003:LinearColor=&MakeLinearColorFromVectorAndFloat(
     "Vector (RGB)"=&MakeVectorFromLinearColorRGB(Color=<rgba 1 1 1 1>)
     "Float (Alpha)"=&Multiply_Float(
      A=&FloatFromCurve(
       FloatCurve=DataInterface() {propertyValues {Curve {keys [<at 0 5>,<at 0.1 1>,<at 0.9 1>,<at 1 0>]}}}
       Scale Curve=&FNTimeOfDayScalar(Night Scale=1, Morning Scale=1, Evening Scale=1)
      )
      B=&FN_PickFloatDependOnPlatform(TargetPlatformActive=1, TargetPlatformNoActive=0.15)
     )
    )) {isSetParameters=true}
   /SetVariables(Particles.Color004:LinearColor=&MakeLinearColorFromVectorAndFloat(
     "Vector (RGB)"=&MakeVectorFromLinearColorRGB(Color=<rgba 0.0844 0.8802 0.9 1.5>)
     "Float (Alpha)"=&Multiply_Float(
      A=&FloatFromCurve(
       FloatCurve=DataInterface() {propertyValues {Curve {keys [<at 0 0.8>,<at 0.85 0.8>,<at 1 0>]}}}
       Scale Curve=&FNTimeOfDayScalar(Night Scale=15, Morning Scale=15, Day Scale=15, Evening Scale=15)
      )
      B=&FN_PickFloatDependOnPlatform(TargetPlatformActive=1, TargetPlatformNoActive=0.6)
     )
    )) {isSetParameters=true}
   &DynamicMaterialParameters(
    Index 0 Param 1=&FloatFromCurve(
     FloatCurve=DataInterface() {propertyValues {Curve {keys [
         {arriveTangent=12.0832, interpMode=RCIM_Cubic, leaveTangent=12.0832, tangentMode=RCTM_User, time=0.7, value=-1}
         {arriveTangent=0.7491, interpMode=RCIM_Cubic, leaveTangent=0.7491, tangentMode=RCTM_User, time=1, value=1}
        ]}}}
     Scale Curve=&RandomRangeFloat(Minimum=0.9, Maximum=1)
    )
    Index 0 Param 2=1
    Index 0 Param 3=1
    Module.Index0Param0Write=true
    Module.Index0Param1Write=true
    Module.Index0Param2Write=true
    Module.Index0Param3Write=true
   )
  ]
  rendererClasses [&NiagaraMeshRendererProperties]
  renderers [
   {rendererClass=&NiagaraMeshRendererProperties, rendererIndex=0}
   {rendererClass=&NiagaraMeshRendererProperties, rendererIndex=1}
   {rendererClass=&NiagaraMeshRendererProperties, rendererIndex=2}
   {rendererClass=&NiagaraMeshRendererProperties, rendererIndex=3}
  ]
 }
}
summary {
 systemName=NS_Emote_YuzuCrank_Laser
 emitters [
  {emitterName=Trail_Bright, rendererClasses [&NiagaraMeshRendererProperties]}
  {
   emitterName=Core_Glow
   rendererClasses [
    &NiagaraSpriteRendererProperties
    &NiagaraMeshRendererProperties
    &NiagaraLightRendererProperties
   ]
  }
  {emitterName=Spawn_Flash, rendererClasses [&NiagaraSpriteRendererProperties]}
 ]
 userVariables [
  User.bIsFrontEnd {type=NiagaraBool}
  User.bIsFrontEndPreview {type=NiagaraBool}
  User.DeltaTimeMax {type=NiagaraFloat, defaultValue=0.0167, description=""}
  User.DustSpawnCurve {type=NiagaraFloat, defaultValue=1, description=""}
  User.RibbonIDOffset {type=NiagaraInt32}
  User.SkeletalMesh {
   type=NiagaraDataInterfaceSkeletalMesh
   defaultValue {
    struct=NiagaraExt_VariableValue_DataInterface
    value {dataInterface=&self:NiagaraDataInterfaceSkeletalMesh_0}
   }
  }
  User.SwingSectionIndex {type=NiagaraInt32}
  User.TrailLifeTimeMult {type=NiagaraFloat, defaultValue=0.4, description=""}
  User.WidthMult {type=NiagaraFloat, defaultValue=2, description=""}
 ]
}
