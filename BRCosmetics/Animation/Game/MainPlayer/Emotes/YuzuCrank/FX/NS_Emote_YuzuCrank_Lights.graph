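# Niagara light effect for the YuzuCrank player emote, providing timed cyan-white lighting.
# A single Light emitter burst-spawns particles placed relative to the system location, with light attributes written each update.
# A float curve scales light color to ramp up, hold, then fade out over time.
# Two light renderers draw the particles, and an extra position attribute offset upward is also computed, apparently for an overhead light.
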NiagaraSystem /BRCosmetics/Animation/Game/MainPlayer/Emotes/YuzuCrank/FX/NS_Emote_YuzuCrank_Lights
dependencies {
 usedDataInterfaces [&NiagaraDataInterfaceCurve]
 usedDynamicInputs [&Multiply_VectorByFloat]
 usedModules [&SystemState]
}
emitters {Light {
  emitterUpdateScript [
   &EmitterState(Life Cycle Mode=::NewEnumerator1, Loop Behavior=::NewEnumerator1, Loop Duration=6, UseLoopDelay=false)
   &SpawnBurst_Instantaneous(Use Spawn Probability=false, UseLoopCountLimit=false)
  ]
  particleSpawnScript [&InitializeParticle(Color Mode=::NewEnumerator0, Lifetime=5, UsePositionOffset=true, Write Material Random=true)]
  particleUpdateScript [
   &ParticleState()
   &SystemLocation(MaskSpawnGroup=false, Offset=<0 0 40>, Offset Coordinate Space=::Simulation)
   &FN_Light_Attributes(
    Color=<rgba 5.4583 7.917 8 1>
    Create Color Attribute=true
    Create Exponent Attribute=true
    Create Position Attribute=true
    Create Radius Attribute=true
    Exponent=5
    Position=<5 0 0>
    Radius=500
    Scale Color=&FloatFromCurve(FloatCurve=DataInterface() {propertyValues {Curve {keys [<at 0 0>,<at 0.21 0.1>,<at 0.26 1>,<at 0.8 1>,<at 0.9 0>]}}})
   )
   /SetVariables(Particles.UpLIghtPos:NiagaraPosition=&AddVectorToPosition(Position=<0 0 0>, Vector=<0 0 250>)) {isSetParameters=true}
  ]
  rendererClasses [&NiagaraLightRendererProperties]
  renderers [
   {rendererClass=&NiagaraLightRendererProperties, rendererIndex=0}
   {rendererClass=&NiagaraLightRendererProperties, rendererIndex=1}
  ]
 }}
summary {
 systemName=NS_Emote_YuzuCrank_Lights
 emitters [{emitterName=Light, rendererClasses [&NiagaraLightRendererProperties]}]
 userVariables [
  User.ReactiveBool {type=NiagaraBool}
  User.ReactiveOn {type=NiagaraFloat, defaultValue=1, description=""}
  User.ReactivityOn {type=NiagaraFloat, defaultValue=1, description=""}
  User.Timestamp {type=NiagaraFloat, defaultValue=1, description=""}
 ]
}
