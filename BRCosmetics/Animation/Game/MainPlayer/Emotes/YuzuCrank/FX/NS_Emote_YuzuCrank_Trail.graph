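# Niagara system for the YuzuCrank emote's red-orange trail, built from one ribbon emitter with four ribbon renderers.
# Particles spawn at a steady rate with custom facing along the system X axis, user multipliers for width and lifetime, and a local leading-edge fix module.
# Curves drive separate edge, bottom-edge and front-glow colors, with a shared TrailAlpha from emitter loop age fading every color layer out.
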NiagaraSystem /BRCosmetics/Animation/Game/MainPlayer/Emotes/YuzuCrank/FX/NS_Emote_YuzuCrank_Trail
dependencies {
 usedDataInterfaces [&NiagaraDataInterfaceCurve, &NiagaraDataInterfaceColorCurve]
 usedDynamicInputs []
 usedModules [&SystemState]
}
emitters {Trail {
  emitterUpdateScript [
   &EmitterState(Life Cycle Mode=::NewEnumerator1, Loop Behavior=::NewEnumerator1, Loop Duration Mode=::NewEnumerator1, UseLoopDelay=false)
   /SetVariables(Emitter.NewSpawnRate=17) {isSetParameters=true}
   &SpawnRate(SpawnRate=$Emitter.NewSpawnRate, Use Spawn Probability=false)
  ]
  particleSpawnScript [&InitializeParticle(
    Color=<rgba 0.6771 0 0 1.5>
    Lifetime=&Multiply_Float(A=1, B=$User.TrailLifeTimeMult)
    Ribbon Facing Mode=::NewEnumerator1
    Ribbon Facing Vector:Vector3f=/Game/Effects/Niagara/DynamicInputs/Engine/FN_GetSystemAxis(Axis:ENiagaraOrientationAxis=::XAxis, Flip Axis:NiagaraBool=false)
    Ribbon Width=&Multiply_Float(A=20, B=$User.WidthMult)
    Ribbon Width Mode=::NewEnumerator1
    UsePositionOffset=false
    Write Material Random=true
   ) {module=InitializeParticle001}]
  particleUpdateScript [
   &ParticleState()
   &self:Leading_Edge_Fix_01() {module=Leading_Edge_Fix_01}
   &ScaleColor(
    Scale Alpha=&FloatFromCurve(FloatCurve=DataInterface() {propertyValues {Curve {keys [<at 0 1>,<at 0.5 0>]}}})
    ScaleA=true
    ScaleRGB=false
   )
   /SetVariables(
    Particles.EdgeColorBot:LinearColor=&ColorFromCurve(ColorCurve=DataInterface() {propertyValues {
       AlphaCurve {keys [
         <at 0 0>
         {arriveTangent=0.2914, interpMode=RCIM_Cubic, leaveTangent=0.2914, time=0.0232, value=0.5}
         <at 0.3431 0.1>
         <at 0.5611 0 RCIM_Cubic>
        ]}
       BlueCurve {keys [<at 0 0.0081 RCIM_Cubic>,<at 0.2884 0.0085 RCIM_Cubic>]}
       GreenCurve {keys [<at 0 0 RCIM_Cubic>,<at 0.2884 0 RCIM_Cubic>]}
       RedCurve {keys [<at 0 0.2917 RCIM_Cubic>,<at 0.2884 0.3073 RCIM_Cubic>]}
      }})
    Particles.EdgeWidthBot:NiagaraFloat=&Multiply_Float(A=21, B=$User.WidthMult)
   ) {isSetParameters=true}
   /SetVariables(
    Particles.EdgeColor:LinearColor=&MultiplyLinearColor(
     Linear Color=&ColorFromCurve(ColorCurve=DataInterface() {propertyValues {
        AlphaCurve {keys [<at 0 0>, {arriveTangent=0.2629, interpMode=RCIM_Cubic, leaveTangent=0.2629, time=0.021, value=5}, <at 0.3804 0.1>, <at 0.5187 0 RCIM_Cubic>]}
        BlueCurve {keys [<at 0.1616 0.405 RCIM_Cubic>,<at 0.3518 0.6 RCIM_Cubic>]}
        GreenCurve {keys [<at 0.1616 2.9516 RCIM_Cubic>,<at 0.3518 2.7768 RCIM_Cubic>]}
        RedCurve {keys [<at 0.1616 15 RCIM_Cubic>,<at 0.3518 15 RCIM_Cubic>]}
       }})
     Scale Factor=&MakeVector4_FromLinearColor(LinearColor=<rgba 1 0.8281 0.8281 1>)
    )
    Particles.EdgeWidth:NiagaraFloat=&Multiply_Float(A=20, B=$User.WidthMult)
   ) {isSetParameters=true}
   /SetVariables(
    Particles.FrontGlowColor:LinearColor=&ColorFromCurve(
     ColorCurve=DataInterface() {propertyValues {
       AlphaCurve {keys [<at 0 2 RCIM_Cubic>,<at 0.0579 0 RCIM_Cubic>]}
       BlueCurve {keys [<at 0 0.6771 RCIM_Cubic>,<at 0.1215 0.0005 RCIM_Cubic>]}
       GreenCurve {keys [<at 0 1.7899 RCIM_Cubic>,<at 0.1215 0 RCIM_Cubic>]}
       RedCurve {keys [<at 0 10 RCIM_Cubic>,<at 0.1215 0.1562 RCIM_Cubic>]}
      }}
     CurveIndex=&Add_Float(
      A=$Particles.NormalizedAge
      B=&Lerp_Float(A=-0.1, B=-0.2, Alpha=&Multiply_Float(A=$Emitter.Age, B=0.2))
     )
    )
    Particles.FrontGlowWidth:NiagaraFloat=&Multiply_Float(A=15, B=$User.WidthMult)
   ) {isSetParameters=true}
   &SolveForcesAndVelocity(Clamp Velocity=false, Limit Acceleration=false)
   /SetVariables(Particles.TrailAlpha:NiagaraFloat=&FloatFromCurve(
     FloatCurve=DataInterface() {propertyValues {Curve {keys [<at 0 1>,<at 0.4177 0>,<at 0.5328 0>,<at 1 0>]}}}
     CurveIndex=$Emitter.NormalizedLoopAge
    )) {isSetParameters=true}
   /SetVariables(
    Particles.Color=&Multiply_LinearColorByFloat(Linear Color=$Particles.Color, Float=$Particles.TrailAlpha)
    Particles.EdgeColor:LinearColor=&Multiply_LinearColorByFloat(Linear Color=$Particles.EdgeColor, Float=$Particles.TrailAlpha)
    Particles.EdgeColorBot:LinearColor=&Multiply_LinearColorByFloat(Linear Color=$Particles.EdgeColorBot, Float=$Particles.TrailAlpha)
    Particles.FrontGlowColor:LinearColor=&Multiply_LinearColorByFloat(Linear Color=$Particles.FrontGlowColor, Float=$Particles.TrailAlpha)
   ) {isSetParameters=true}
  ]
  rendererClasses [&NiagaraRibbonRendererProperties]
  renderers [
   {rendererClass=&NiagaraRibbonRendererProperties}
   {rendererClass=&NiagaraRibbonRendererProperties, rendererIndex=1}
   {rendererClass=&NiagaraRibbonRendererProperties, rendererIndex=2}
   {rendererClass=&NiagaraRibbonRendererProperties, rendererIndex=3}
  ]
 }}
summary {
 systemName=NS_Emote_YuzuCrank_Trail
 emitters [{emitterName=Trail, rendererClasses [&NiagaraRibbonRendererProperties]}]
 userVariables [
  User.Anim_LocalSpaceInheritance {type=NiagaraFloat, defaultValue=0.5, description=""}
  User.bIsFrontEnd {type=NiagaraBool}
  User.bIsFrontEndPreview {type=NiagaraBool}
  User.DustSpawnCurve {type=NiagaraFloat, defaultValue=1, description=""}
  User.RibbonIDOffset {type=NiagaraInt32}
  User.SkeletalMesh {
   type=NiagaraDataInterfaceSkeletalMesh
   defaultValue {
    struct=NiagaraExt_VariableValue_DataInterface
    value {dataInterface=&self:NiagaraDataInterfaceSkeletalMesh_0}
   }
  }
  User.SwingSectionIndex {type=NiagaraInt32}
  User.TrailLifeTimeMult {type=NiagaraFloat, defaultValue=0.4, description=""}
  User.WidthMult {type=NiagaraFloat, defaultValue=2, description=""}
 ]
}
